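# Physics asset for the crouch variant of the male medium Hip Hare body cosmetic, covering its dynamic jacket and shirt bones.
# Seven jacket panels of thin boxes, one segment at the front and two at sides and back, chain from master bodies linked to the pelvis.
# A four-box shirt chain hangs from spine_05 via its master.
# Chain joints have angular limits, while master, spine and thigh joints swing and twist freely among capsule body shapes.
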PhysicsAsset /BRCosmetics/Characters/Player/Male/Medium/Bodies/M_MED_Hip_Hare/Meshes/M_MED_Hip_Hare_Crouch_Physics
bodies {
 Dyn_Shirt_1 [{shapeType=Box, extentX=3, extentY=1, extentZ=10, center=<0.5 -0.715 0>}]
 Dyn_Shirt_2 [{shapeType=Box, extentX=6, extentY=2, extentZ=10, center=<1 -0.4681 0.0155>}]
 Dyn_Shirt_3 [{shapeType=Box, extentX=6, extentY=2, extentZ=10, center=<1 0 0>}]
 Dyn_Shirt_4 [{shapeType=Box, extentX=6, extentY=2, extentZ=10, center=<1 0 0>}]
 Dyn_Shirt_Master [{shapeType=Box, extentX=1, extentY=1, extentZ=1}]
 dyn_jacket_bk_01 [{shapeType=Box, extentX=7.6, extentY=14.25, extentZ=0.8, center=<3.2 0 -0.5>}]
 dyn_jacket_bk_02 [{shapeType=Box, extentX=7.6, extentY=14.25, extentZ=0.8, center=<3.9 0 -0.5>}]
 dyn_jacket_bk_MSTR [{shapeType=Capsule, length=1, radius=0.5}]
 dyn_jacket_bk_l_01 [{shapeType=Box, extentX=7.6, extentY=14.25, extentZ=0.8, center=<3.2 0 0>}]
 dyn_jacket_bk_l_02 [{shapeType=Box, extentX=7.6, extentY=14.25, extentZ=0.8, center=<3.9 0 0>}]
 dyn_jacket_bk_l_MSTR [{shapeType=Capsule, length=1, radius=0.5}]
 dyn_jacket_bk_r_01 [{shapeType=Box, extentX=7.6, extentY=14.25, extentZ=0.8, center=<3.2 0 -0.5>}]
 dyn_jacket_bk_r_02 [{shapeType=Box, extentX=7.6, extentY=14.25, extentZ=0.8, center=<3.9 0 -0.5>}]
 dyn_jacket_bk_r_MSTR [{shapeType=Capsule, length=1, radius=0.5}]
 dyn_jacket_fr_l_01 [{shapeType=Box, extentX=15, extentY=14, extentZ=2, center=<8 0 1.2334>, rotation=<rot -4.9999 0 0>}]
 dyn_jacket_fr_l_MSTR [{shapeType=Capsule, length=1, radius=0.5}]
 dyn_jacket_fr_r_01 [{shapeType=Box, extentX=15, extentY=14, extentZ=1, center=<8.0249 0 0.8548>, rotation=<rot 5.0001 0 0>}]
 dyn_jacket_fr_r_MSTR [{shapeType=Capsule, length=1, radius=0.5}]
 dyn_jacket_l_01 [{shapeType=Box, extentX=7.6, extentY=14.25, extentZ=0.8, center=<3.2 0 1.5869>}]
 dyn_jacket_l_02 [{shapeType=Box, extentX=7.6, extentY=14.25, extentZ=0.8, center=<3.9 -0.0001 0.9113>}]
 dyn_jacket_l_MSTR [{shapeType=Capsule, length=1, radius=0.5}]
 dyn_jacket_r_01 [{shapeType=Box, extentX=7.6, extentY=14.25, extentZ=0.8, center=<3.2 0 -1.092>}]
 dyn_jacket_r_02 [{shapeType=Box, extentX=6, extentY=14.25, extentZ=0.8, center=<3.9 0 -0.5>}]
 dyn_jacket_r_MSTR [{shapeType=Capsule, length=1, radius=0.5}]
 pelvis [{shapeType=Capsule, length=6.7, radius=9, center=<-3 0.3098 0>}]
 spine_05 [
  {shapeType=Capsule, length=10, radius=13, center=<-2.8983 -0.5102 0.0002>}
  {shapeType=Capsule, length=15, radius=10, center=<-14.9353 3.5955 0>, rotation=<rot 90 0 10.0004>}
 ]
 thigh_l [{shapeType=Capsule, length=24.5, radius=9.95, center=<-13.0229 1.1978 0>, rotation=<rot 90 6.2419 1.2421>}]
 thigh_r [
  {shapeType=Capsule, length=27, radius=7.25, center=<14.8465 -0.8269 0.0682>, rotation=<rot 90 0 0>}
  {shapeType=Capsule, length=27, radius=7.25, center=<14.8465 -2.0872 0.0682>, rotation=<rot 90 0 0>}
 ]
}
constraints [
 {bone1Name=dyn_jacket_l_02, bone2Name=dyn_jacket_l_01, swing1LimitDegrees=60, swing2LimitDegrees=10, twistLimitDegrees=50}
 {bone1Name=dyn_jacket_bk_l_02, bone2Name=dyn_jacket_bk_l_01, swing1LimitDegrees=60, swing2LimitDegrees=60, twistLimitDegrees=5}
 {bone1Name=dyn_jacket_bk_02, bone2Name=dyn_jacket_bk_01, swing1LimitDegrees=60, swing2LimitDegrees=60, twistLimitDegrees=5}
 {bone1Name=dyn_jacket_bk_r_02, bone2Name=dyn_jacket_bk_r_01, swing1LimitDegrees=60, swing2LimitDegrees=60, twistLimitDegrees=5}
 {bone1Name=dyn_jacket_r_02, bone2Name=dyn_jacket_r_01, swing1LimitDegrees=60, swing2LimitDegrees=60, twistLimitDegrees=5}
 {bone1Name=dyn_jacket_l_01, bone2Name=dyn_jacket_l_MSTR, swing1LimitDegrees=60, swing2LimitDegrees=10, twistLimitDegrees=50}
 {bone1Name=dyn_jacket_fr_l_01, bone2Name=dyn_jacket_fr_l_MSTR, swing1LimitDegrees=15, swing2LimitDegrees=60, twistLimitDegrees=10}
 {bone1Name=dyn_jacket_bk_l_01, bone2Name=dyn_jacket_bk_l_MSTR, swing1LimitDegrees=15, swing2LimitDegrees=15, twistLimitDegrees=1}
 {bone1Name=dyn_jacket_bk_01, bone2Name=dyn_jacket_bk_MSTR, swing1LimitDegrees=60, swing2LimitDegrees=60, twistLimitDegrees=5}
 {bone1Name=dyn_jacket_fr_r_01, bone2Name=dyn_jacket_fr_r_MSTR, swing1LimitDegrees=15, swing2LimitDegrees=60, twistLimitDegrees=2}
 {bone1Name=dyn_jacket_bk_r_01, bone2Name=dyn_jacket_bk_r_MSTR, swing1LimitDegrees=60, swing2LimitDegrees=60, twistLimitDegrees=5}
 {bone1Name=dyn_jacket_r_01, bone2Name=dyn_jacket_r_MSTR, swing1LimitDegrees=15, swing2LimitDegrees=15, twistLimitDegrees=1}
 {
  bone1Name=thigh_l
  bone2Name=pelvis
  swing1LimitDegrees=45
  swing1Motion=Free
  swing2LimitDegrees=45
  swing2Motion=Free
  twistLimitDegrees=45
  twistMotion=Free
 }
 {
  bone1Name=thigh_r
  bone2Name=pelvis
  swing1LimitDegrees=45
  swing1Motion=Free
  swing2LimitDegrees=45
  swing2Motion=Free
  twistLimitDegrees=45
  twistMotion=Free
 }
 {
  bone1Name=spine_05
  bone2Name=pelvis
  swing1LimitDegrees=45
  swing1Motion=Free
  swing2LimitDegrees=45
  swing2Motion=Free
  twistLimitDegrees=45
  twistMotion=Free
 }
 {
  bone1Name=Dyn_Shirt_Master
  bone2Name=spine_05
  swing1LimitDegrees=45
  swing1Motion=Free
  swing2LimitDegrees=45
  swing2Motion=Free
  twistLimitDegrees=45
  twistMotion=Free
 }
 {bone1Name=Dyn_Shirt_1, bone2Name=Dyn_Shirt_Master, swing1LimitDegrees=10, swing2LimitDegrees=20, twistLimitDegrees=1}
 {bone1Name=Dyn_Shirt_2, bone2Name=Dyn_Shirt_1, swing1LimitDegrees=10, swing2LimitDegrees=20, twistLimitDegrees=1}
 {bone1Name=Dyn_Shirt_3, bone2Name=Dyn_Shirt_2, swing1LimitDegrees=10, swing2LimitDegrees=20, twistLimitDegrees=1}
 {bone1Name=Dyn_Shirt_4, bone2Name=Dyn_Shirt_3, swing1LimitDegrees=10, swing2LimitDegrees=20, twistLimitDegrees=1}
 {
  bone1Name=dyn_jacket_fr_l_MSTR
  bone2Name=pelvis
  swing1LimitDegrees=45
  swing1Motion=Free
  swing2LimitDegrees=45
  swing2Motion=Free
  twistLimitDegrees=45
  twistMotion=Free
 }
 {
  bone1Name=dyn_jacket_l_MSTR
  bone2Name=pelvis
  swing1LimitDegrees=45
  swing1Motion=Free
  swing2LimitDegrees=45
  swing2Motion=Free
  twistLimitDegrees=45
  twistMotion=Free
 }
 {
  bone1Name=dyn_jacket_bk_l_MSTR
  bone2Name=pelvis
  swing1LimitDegrees=45
  swing1Motion=Free
  swing2LimitDegrees=45
  swing2Motion=Free
  twistLimitDegrees=45
  twistMotion=Free
 }
 {
  bone1Name=dyn_jacket_bk_MSTR
  bone2Name=pelvis
  swing1LimitDegrees=45
  swing1Motion=Free
  swing2LimitDegrees=45
  swing2Motion=Free
  twistLimitDegrees=45
  twistMotion=Free
 }
 {
  bone1Name=dyn_jacket_fr_r_MSTR
  bone2Name=pelvis
  swing1LimitDegrees=45
  swing1Motion=Free
  swing2LimitDegrees=45
  swing2Motion=Free
  twistLimitDegrees=45
  twistMotion=Free
 }
 {
  bone1Name=dyn_jacket_bk_r_MSTR
  bone2Name=pelvis
  swing1LimitDegrees=45
  swing1Motion=Free
  swing2LimitDegrees=45
  swing2Motion=Free
  twistLimitDegrees=45
  twistMotion=Free
 }
 {
  bone1Name=dyn_jacket_r_MSTR
  bone2Name=pelvis
  swing1LimitDegrees=45
  swing1Motion=Free
  swing2LimitDegrees=45
  swing2Motion=Free
  twistLimitDegrees=45
  twistMotion=Free
 }
]
