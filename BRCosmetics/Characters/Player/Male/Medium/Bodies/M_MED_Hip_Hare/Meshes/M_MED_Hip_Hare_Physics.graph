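# Physics asset for the Hip Hare medium male player body, defining secondary motion for its jacket and shirt bones.
# Seven two-segment jacket chains of thin boxes hang from master bodies on the pelvis, and a four-box shirt chain hangs from spine_05.
# Capsules on the pelvis, spine and thighs form the body volume.
# Cloth segment joints have per-chain swing and tight twist limits, while body and master joints rotate freely.
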PhysicsAsset /BRCosmetics/Characters/Player/Male/Medium/Bodies/M_MED_Hip_Hare/Meshes/M_MED_Hip_Hare_Physics
bodies {
 Dyn_Shirt_1 [{shapeType=Box, extentX=4, extentY=2, extentZ=10, center=<2 0 0>}]
 Dyn_Shirt_2 [{shapeType=Box, extentX=7, extentY=2, extentZ=15, center=<3.5 0 0>}]
 Dyn_Shirt_3 [{shapeType=Box, extentX=6, extentY=2, extentZ=15, center=<3 0 0>}]
 Dyn_Shirt_4 [{shapeType=Box, extentX=5, extentY=2, extentZ=10, center=<2.5 0 0>}]
 Dyn_Shirt_Master [{shapeType=Capsule, length=0.505, radius=0.505, shapeName=Dyn_Shirt_Master_capsule}]
 dyn_jacket_bk_01 [{shapeType=Box, extentX=7.6, extentY=14.25, extentZ=0.8, center=<3.2 0 -0.5133>}]
 dyn_jacket_bk_02 [{shapeType=Box, extentX=7.6, extentY=14.25, extentZ=0.8, center=<3.9 0 -0.8768>}]
 dyn_jacket_bk_MSTR [{shapeType=Capsule, length=1, radius=0.5}]
 dyn_jacket_bk_l_01 [{shapeType=Box, extentX=7.6, extentY=14.25, extentZ=0.5, center=<3.1998 0 0.2539>}]
 dyn_jacket_bk_l_02 [{shapeType=Box, extentX=7.6, extentY=14.25, extentZ=1, center=<3.9 0 -0.7595>}]
 dyn_jacket_bk_l_MSTR [{shapeType=Capsule, length=1, radius=0.5}]
 dyn_jacket_bk_r_01 [{shapeType=Box, extentX=7.6, extentY=14.25, extentZ=0.8, center=<3.1998 0 0.2539>}]
 dyn_jacket_bk_r_02 [{shapeType=Box, extentX=7.6, extentY=14.25, extentZ=0.8, center=<3.9 0 -0.7595>}]
 dyn_jacket_bk_r_MSTR [{shapeType=Capsule, length=1, radius=0.5}]
 dyn_jacket_fr_l_01 [{shapeType=Box, extentX=8, extentY=14, extentZ=0.5, center=<3.0912 0 -0.4763>, rotation=<rot 0.0002 0.0001 0>}]
 dyn_jacket_fr_l_02 [{shapeType=Box, extentX=8, extentY=14, extentZ=1.3, center=<1.0795 0 -0.2321>, rotation=<rot 0.0004 0.0002 0>}]
 dyn_jacket_fr_l_MSTR [{shapeType=Capsule, length=1, radius=0.5}]
 dyn_jacket_fr_r_01 [{shapeType=Box, extentX=8, extentY=14, extentZ=0.5, center=<3 0 0>}]
 dyn_jacket_fr_r_02 [{shapeType=Box, extentX=8, extentY=14, extentZ=1.5, center=<1 0 -0.5>}]
 dyn_jacket_fr_r_MSTR [{shapeType=Capsule, length=1, radius=0.5}]
 dyn_jacket_l_01 [{shapeType=Box, extentX=7.6, extentY=14.25, extentZ=0.8, center=<3.2 0 0.0644>}]
 dyn_jacket_l_02 [{shapeType=Box, extentX=7.6, extentY=14.25, extentZ=0.8, center=<3.9 0 -1.5>}]
 dyn_jacket_l_MSTR [{shapeType=Capsule, length=1, radius=0.5}]
 dyn_jacket_r_01 [{shapeType=Box, extentX=7.6, extentY=14.25, extentZ=0.8, center=<3.2 0 -0.2>}]
 dyn_jacket_r_02 [{shapeType=Box, extentX=7.6, extentY=14.25, extentZ=0.8, center=<3.9 0 -0.5>}]
 dyn_jacket_r_MSTR [{shapeType=Capsule, length=1, radius=0.5}]
 pelvis [{shapeType=Capsule, length=6.7, radius=9, center=<-3 0.3098 0>}]
 spine_02 [{shapeType=Capsule, length=10, radius=14.7, shapeName=spine_02_capsule, center=<3.6925 0 0>}]
 spine_03 [{shapeType=Capsule, length=10, radius=13.4, shapeName=spine_03_capsule, center=<6 0 0>}]
 spine_04 [{shapeType=Capsule, length=10, radius=12.2, shapeName=spine_04_capsule, center=<5 0 0>}]
 spine_05 [{shapeType=Capsule, length=10, radius=10, shapeName=spine_05_capsule, center=<6.0188 0 0>}]
 thigh_l [{shapeType=Capsule, length=25.5, radius=7.25, center=<-12.9407 1.1 0.8357>, rotation=<rot 90 89.6476 91.737>}]
 thigh_r [{shapeType=Capsule, length=25.75, radius=7, center=<12.9234 -0.7589 -0.8357>, rotation=<rot 90 9.7524 11.8408>}]
}
constraints [
 {bone1Name=dyn_jacket_l_02, bone2Name=dyn_jacket_l_01, swing1LimitDegrees=60, swing2LimitDegrees=60, twistLimitDegrees=5}
 {bone1Name=dyn_jacket_fr_l_02, bone2Name=dyn_jacket_fr_l_01, swing1LimitDegrees=15, swing2LimitDegrees=30, twistLimitDegrees=1}
 {bone1Name=dyn_jacket_bk_l_02, bone2Name=dyn_jacket_bk_l_01, swing1LimitDegrees=60, swing2LimitDegrees=60, twistLimitDegrees=5}
 {bone1Name=dyn_jacket_bk_02, bone2Name=dyn_jacket_bk_01, swing1LimitDegrees=60, swing2LimitDegrees=60, twistLimitDegrees=5}
 {bone1Name=dyn_jacket_fr_r_02, bone2Name=dyn_jacket_fr_r_01, swing1LimitDegrees=1, swing2LimitDegrees=15, twistLimitDegrees=1}
 {bone1Name=dyn_jacket_bk_r_02, bone2Name=dyn_jacket_bk_r_01, swing1LimitDegrees=60, swing2LimitDegrees=60, twistLimitDegrees=5}
 {bone1Name=dyn_jacket_r_02, bone2Name=dyn_jacket_r_01, swing1LimitDegrees=60, swing2LimitDegrees=60, twistLimitDegrees=5}
 {bone1Name=dyn_jacket_l_01, bone2Name=dyn_jacket_l_MSTR, swing1LimitDegrees=25, swing2LimitDegrees=20, twistLimitDegrees=1}
 {bone1Name=dyn_jacket_fr_l_01, bone2Name=dyn_jacket_fr_l_MSTR, swing1LimitDegrees=15, swing2LimitDegrees=45, twistLimitDegrees=1}
 {bone1Name=dyn_jacket_bk_l_01, bone2Name=dyn_jacket_bk_l_MSTR, swing1LimitDegrees=15, swing2LimitDegrees=15, twistLimitDegrees=5}
 {bone1Name=dyn_jacket_bk_01, bone2Name=dyn_jacket_bk_MSTR, swing1LimitDegrees=60, swing2LimitDegrees=60, twistLimitDegrees=5}
 {bone1Name=dyn_jacket_fr_r_01, bone2Name=dyn_jacket_fr_r_MSTR, swing1LimitDegrees=1, swing2LimitDegrees=15, twistLimitDegrees=1}
 {bone1Name=dyn_jacket_bk_r_01, bone2Name=dyn_jacket_bk_r_MSTR, swing1LimitDegrees=15, swing2LimitDegrees=15, twistLimitDegrees=5}
 {bone1Name=dyn_jacket_r_01, bone2Name=dyn_jacket_r_MSTR, swing1LimitDegrees=15, swing2LimitDegrees=15, twistLimitDegrees=1}
 {
  bone1Name=thigh_l
  bone2Name=pelvis
  swing1LimitDegrees=45
  swing1Motion=Free
  swing2LimitDegrees=45
  swing2Motion=Free
  twistLimitDegrees=45
  twistMotion=Free
 }
 {
  bone1Name=thigh_r
  bone2Name=pelvis
  swing1LimitDegrees=45
  swing1Motion=Free
  swing2LimitDegrees=45
  swing2Motion=Free
  twistLimitDegrees=45
  twistMotion=Free
 }
 {
  bone1Name=Dyn_Shirt_Master
  bone2Name=spine_05
  swing1LimitDegrees=45
  swing1Motion=Free
  swing2LimitDegrees=45
  swing2Motion=Free
  twistLimitDegrees=45
  twistMotion=Free
 }
 {bone1Name=Dyn_Shirt_1, bone2Name=Dyn_Shirt_Master, swing1LimitDegrees=10, swing2LimitDegrees=20, twistLimitDegrees=1}
 {bone1Name=Dyn_Shirt_2, bone2Name=Dyn_Shirt_1, swing1LimitDegrees=10, swing2LimitDegrees=20, twistLimitDegrees=1}
 {bone1Name=Dyn_Shirt_3, bone2Name=Dyn_Shirt_2, swing1LimitDegrees=10, swing2LimitDegrees=20, twistLimitDegrees=1}
 {bone1Name=Dyn_Shirt_4, bone2Name=Dyn_Shirt_3, swing1LimitDegrees=10, swing2LimitDegrees=20, twistLimitDegrees=1}
 {
  bone1Name=dyn_jacket_fr_l_MSTR
  bone2Name=pelvis
  swing1LimitDegrees=45
  swing1Motion=Free
  swing2LimitDegrees=45
  swing2Motion=Free
  twistLimitDegrees=45
  twistMotion=Free
 }
 {
  bone1Name=dyn_jacket_l_MSTR
  bone2Name=pelvis
  swing1LimitDegrees=45
  swing1Motion=Free
  swing2LimitDegrees=45
  swing2Motion=Free
  twistLimitDegrees=45
  twistMotion=Free
 }
 {
  bone1Name=dyn_jacket_bk_l_MSTR
  bone2Name=pelvis
  swing1LimitDegrees=45
  swing1Motion=Free
  swing2LimitDegrees=45
  swing2Motion=Free
  twistLimitDegrees=45
  twistMotion=Free
 }
 {
  bone1Name=dyn_jacket_bk_MSTR
  bone2Name=pelvis
  swing1LimitDegrees=45
  swing1Motion=Free
  swing2LimitDegrees=45
  swing2Motion=Free
  twistLimitDegrees=45
  twistMotion=Free
 }
 {
  bone1Name=dyn_jacket_fr_r_MSTR
  bone2Name=pelvis
  swing1LimitDegrees=45
  swing1Motion=Free
  swing2LimitDegrees=45
  swing2Motion=Free
  twistLimitDegrees=45
  twistMotion=Free
 }
 {
  bone1Name=dyn_jacket_bk_r_MSTR
  bone2Name=pelvis
  swing1LimitDegrees=45
  swing1Motion=Free
  swing2LimitDegrees=45
  swing2Motion=Free
  twistLimitDegrees=45
  twistMotion=Free
 }
 {
  bone1Name=dyn_jacket_r_MSTR
  bone2Name=pelvis
  swing1LimitDegrees=45
  swing1Motion=Free
  swing2LimitDegrees=45
  swing2Motion=Free
  twistLimitDegrees=45
  twistMotion=Free
 }
 {
  bone1Name=spine_02
  bone2Name=pelvis
  swing1LimitDegrees=45
  swing1Motion=Free
  swing2LimitDegrees=45
  swing2Motion=Free
  twistLimitDegrees=45
  twistMotion=Free
 }
 {
  bone1Name=spine_03
  bone2Name=spine_02
  swing1LimitDegrees=45
  swing1Motion=Free
  swing2LimitDegrees=45
  swing2Motion=Free
  twistLimitDegrees=45
  twistMotion=Free
 }
 {
  bone1Name=spine_04
  bone2Name=spine_03
  swing1LimitDegrees=45
  swing1Motion=Free
  swing2LimitDegrees=45
  swing2Motion=Free
  twistLimitDegrees=45
  twistMotion=Free
 }
 {
  bone1Name=spine_05
  bone2Name=spine_04
  swing1LimitDegrees=45
  swing1Motion=Free
  swing2LimitDegrees=45
  swing2Motion=Free
  twistLimitDegrees=45
  twistMotion=Free
 }
]
