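# Animation Blueprint for the Hip Hare player body that drives secondary motion and clipping-style corrections on its jacket and shirt bones.
# Separate standing and crouch chains each run RigidBody simulation in spine_01 space with their own physics asset and gravity.
# Long stacks of rotation constraints bind jacket flaps to the thighs, some pinned to the pelvis.
# Additive bone offsets, some driven by leg-raise, jump, fall, crouch and dive variables, adjust the result.
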AnimBlueprint /BRCosmetics/Characters/Player/Male/Medium/Bodies/M_MED_Hip_Hare/Meshes/M_MED_Hip_Hare_AnimBP
target_skeleton=&/M_MED_Hip_Hare_Skeleton
graphs [
 AnimGraph {nodes [
   n1 = "Constraint - dyn_jacket_l_MSTR (R)"(Alpha=1, ConstraintWeights_0=1) {
    alphaBoolBlend {blendOption=Linear}
    boneToModify {boneName=dyn_jacket_l_MSTR}
    constraintSetup [{transformType=Rotation, targetBone {boneName=thigh_l}}]
    constraintWeights [1]
   }
   n2 = "Constraint - dyn_jacket_l_01 (R)"(ComponentPose=@n1, Alpha=1, ConstraintWeights_0=1) {
    alphaBoolBlend {blendOption=Linear}
    boneToModify {boneName=dyn_jacket_l_01}
    constraintSetup [{transformType=Rotation, targetBone {boneName=thigh_l}}]
    constraintWeights [1]
   }
   "Constraint - dyn_jacket_l_02 (R)"(ComponentPose=@n2, Alpha=1, ConstraintWeights_0=1) {
    alphaBoolBlend {blendOption=Linear}
    boneToModify {boneName=dyn_jacket_l_02}
    constraintSetup [{transformType=Rotation, targetBone {boneName=thigh_l}}]
    constraintWeights [1]
   }
   n3 = "Constraint - dyn_jacket_l_MSTR (R)"(Alpha=1, ConstraintWeights_0=1) {
    alphaBoolBlend {blendOption=Linear}
    boneToModify {boneName=dyn_jacket_l_MSTR}
    constraintSetup [{transformType=Rotation, targetBone {boneName=thigh_l}}]
    constraintWeights [1]
   }
   n4 = "Constraint - dyn_jacket_l_01 (R)"(ComponentPose=@n3, Alpha=1, ConstraintWeights_0=1) {
    alphaBoolBlend {blendOption=Linear}
    boneToModify {boneName=dyn_jacket_l_01}
    constraintSetup [{transformType=Rotation, targetBone {boneName=thigh_l}}]
    constraintWeights [1]
   }
   "Constraint - dyn_jacket_l_02 (R)"(ComponentPose=@n4, Alpha=1, ConstraintWeights_0=1) {
    alphaBoolBlend {blendOption=Linear}
    boneToModify {boneName=dyn_jacket_l_02}
    constraintSetup [{transformType=Rotation, targetBone {boneName=thigh_l}}]
    constraintWeights [1]
   }
   "Transform (Modify) Bone - Bone: dyn_jacket_l_MSTR"(Alpha=1, Translation="0,0,0", Rotation="0,0,0", Scale="1,1,1") {
    rotationMode=BMM_Additive
    rotationSpace=BCS_BoneSpace
    translationMode=BMM_Additive
    translationSpace=BCS_BoneSpace
    alphaBoolBlend {blendOption=Linear}
    alphaScaleBiasClamp {bInterpResult=true}
    boneToModify {boneName=dyn_jacket_l_MSTR}
   }
   n5 = "Transform (Modify) Bone - Bone: dyn_jacket_fr_l_MSTR"(
    ComponentPose=Local To Component(LocalPose=Use cached pose 'Crouch'())
    Alpha=Get legUp_l() -> (legUp_l:double)
    Translation="-7,-5,-2"
    Rotation="50,0,0"
    Scale="1,1,1"
   ) {
    rotationMode=BMM_Additive
    rotationSpace=BCS_BoneSpace
    scaleSpace=BCS_BoneSpace
    translationMode=BMM_Additive
    translationSpace=BCS_BoneSpace
    alphaBoolBlend {blendInTime=0.2, blendOption=Linear, blendOutTime=0.2}
    boneToModify {boneName=dyn_jacket_fr_l_MSTR}
    rotation {pitch=50}
    translation=<-7 -5 -2>
   }
   n6 = "Transform (Modify) Bone - Bone: dyn_jacket_fr_l_01"(ComponentPose=@n5, Alpha=Get legUp_l(), Translation="0,0,5", Rotation="30,0,0", Scale="1,1,1") {
    rotationMode=BMM_Additive
    rotationSpace=BCS_BoneSpace
    scaleSpace=BCS_BoneSpace
    translationMode=BMM_Additive
    translationSpace=BCS_BoneSpace
    alphaBoolBlend {blendInTime=0.2, blendOption=Linear, blendOutTime=0.2}
    boneToModify {boneName=dyn_jacket_fr_l_01}
    rotation {pitch=30}
    translation {z=5}
   }
   n7 = "Transform (Modify) Bone - Bone: dyn_jacket_fr_l_02"(ComponentPose=@n6, Alpha=Get legUp_l(), Translation="0,0,5", Rotation="0,0,0", Scale="1,1,1") {
    rotationMode=BMM_Additive
    rotationSpace=BCS_BoneSpace
    scaleSpace=BCS_BoneSpace
    translationMode=BMM_Additive
    translationSpace=BCS_BoneSpace
    alphaBoolBlend {blendInTime=0.2, blendOption=Linear, blendOutTime=0.2}
    boneToModify {boneName=dyn_jacket_fr_l_02}
    translation {z=5}
   }
   n8 = "Transform (Modify) Bone - Bone: dyn_jacket_fr_r_MSTR"(ComponentPose=@n7, Alpha=Get legUp_r() -> (legUp_r:double), Translation="-7,5,-2", Rotation="50,0,0", Scale="1,1,1") {
    rotationMode=BMM_Additive
    rotationSpace=BCS_BoneSpace
    scaleSpace=BCS_BoneSpace
    translationMode=BMM_Additive
    translationSpace=BCS_BoneSpace
    alphaBoolBlend {blendInTime=0.2, blendOption=Linear, blendOutTime=0.2}
    boneToModify {boneName=dyn_jacket_fr_r_MSTR}
    rotation {pitch=50}
    translation=<-7 5 -2>
   }
   n9 = "Transform (Modify) Bone - Bone: dyn_jacket_r_MSTR"!(ComponentPose=@n8, bAlphaBoolEnabled:bool=Get bIsJumping(), Translation="1.5,1.5,1", Rotation="0,0,0", Scale="1,1,1") -> (Pose) {
    alphaInputType=Bool
    rotationMode=BMM_Additive
    rotationSpace=BCS_BoneSpace
    scaleSpace=BCS_BoneSpace
    translationMode=BMM_Additive
    translationSpace=BCS_BoneSpace
    alphaBoolBlend {blendInTime=0.2, blendOption=Linear, blendOutTime=0.2}
    boneToModify {boneName=dyn_jacket_r_MSTR}
    translation=<1.5 1.5 1>
   }
   n10 = Get bIsFalling()
   n11 = "Transform (Modify) Bone - Bone: dyn_jacket_fr_r_MSTR"!(ComponentPose=@n9, bAlphaBoolEnabled:bool=@n10, Translation="1.5,1.5,1", Rotation="0,0,0", Scale="1,1,1") -> (Pose) {
    alphaInputType=Bool
    rotationMode=BMM_Additive
    rotationSpace=BCS_BoneSpace
    scaleSpace=BCS_BoneSpace
    translationMode=BMM_Additive
    translationSpace=BCS_BoneSpace
    alphaBoolBlend {blendInTime=0.2, blendOption=Linear, blendOutTime=0.2}
    boneToModify {boneName=dyn_jacket_fr_r_MSTR}
    translation=<1.5 1.5 1>
   }
   n12 = "Transform (Modify) Bone - Bone: dyn_jacket_r_MSTR"(ComponentPose=@n11, bAlphaBoolEnabled=@n10, Translation="1.5,1.5,1", Rotation="0,0,0", Scale="1,1,1") {
    alphaInputType=Bool
    rotationMode=BMM_Additive
    rotationSpace=BCS_BoneSpace
    scaleSpace=BCS_BoneSpace
    translationMode=BMM_Additive
    translationSpace=BCS_BoneSpace
    alphaBoolBlend {blendInTime=0.2, blendOption=Linear, blendOutTime=0.2}
    boneToModify {boneName=dyn_jacket_r_MSTR}
    translation=<1.5 1.5 1>
   }
   n13 = "Transform (Modify) Bone - Bone: Dyn_Shirt_Master"(ComponentPose=@n12, Alpha=1, Translation="0,-1,0", Rotation="0,0,0", Scale="1,1,1") {
    rotationMode=BMM_Additive
    rotationSpace=BCS_BoneSpace
    scaleSpace=BCS_BoneSpace
    translationMode=BMM_Additive
    translationSpace=BCS_BoneSpace
    alphaBoolBlend {blendInTime=0.2, blendOption=Linear, blendOutTime=0.2}
    boneToModify {boneName=Dyn_Shirt_Master}
    translation {y=-1}
   }
   n14 = "Transform (Modify) Bone - Bone: dyn_jacket_fr_r_MSTR"(ComponentPose=@n13, bAlphaBoolEnabled=True, Translation="-5,3,0", Rotation="0,0,0", Scale="1,1,1") {
    alphaInputType=Bool
    rotationMode=BMM_Additive
    rotationSpace=BCS_BoneSpace
    scaleSpace=BCS_BoneSpace
    translationMode=BMM_Additive
    translationSpace=BCS_BoneSpace
    alphaBoolBlend {blendInTime=0.2, blendOption=Linear, blendOutTime=0.5}
    boneToModify {boneName=dyn_jacket_fr_r_MSTR}
    translation=<-5 3>
   }
   n15 = RigidBody!(ComponentPose=@n14, Alpha, OverrideWorldGravity:Vector=Get Gravity Override(), ExternalForce) -> (Pose) {
    bOverrideWorldGravity=true
    overridePhysicsAsset=&/M_MED_Hip_Hare_Physics
    simulationSpace=BaseBoneSpace
    alphaBoolBlend {blendOption=Linear}
    alphaScaleBiasClamp {bMapRange=true}
    baseBoneRef {boneName=spine_01}
    componentAppliedLinearAccClamp=<1200 1200 1200>
    componentLinearAccScale=<0.4 0.3 0.3>
    componentLinearVelScale=<0.4 0.3 0.3>
    overrideWorldGravity {z=-980}
   }
   n16 = "Transform (Modify) Bone - Bone: Dyn_Shirt_3"(ComponentPose=@n15, Alpha=1, Translation="0,-1,0", Rotation="0,0,0", Scale="1,1,1") {
    rotationMode=BMM_Additive
    rotationSpace=BCS_BoneSpace
    scaleSpace=BCS_BoneSpace
    translationMode=BMM_Additive
    translationSpace=BCS_BoneSpace
    alphaBoolBlend {blendInTime=0.2, blendOption=Linear, blendOutTime=0.2}
    boneToModify {boneName=Dyn_Shirt_3}
    translation {y=-1}
   }
   n17 = "Transform (Modify) Bone - Bone: dyn_jacket_bk_l_MSTR"(ComponentPose=@n16, Alpha=1, Translation="0,0,0", Rotation="20,0,0", Scale="1,1,1") {
    rotationMode=BMM_Additive
    rotationSpace=BCS_BoneSpace
    alphaBoolBlend {blendOption=Linear}
    alphaScaleBiasClamp {bInterpResult=true, interpSpeedIncreasing=20}
    boneToModify {boneName=dyn_jacket_bk_l_MSTR}
    rotation {pitch=20}
   }
   n18 = "Constraint - dyn_jacket_fr_r_MSTR (R)"(ComponentPose=@n17, Alpha=1, ConstraintWeights_0=1) {
    alphaBoolBlend {blendOption=Linear}
    boneToModify {boneName=dyn_jacket_fr_r_MSTR}
    constraintSetup [{transformType=Rotation, targetBone {boneName=thigh_r}}]
    constraintWeights [1]
   }
   n19 = "Constraint - dyn_jacket_fr_r_01 (R)"(ComponentPose=@n18, Alpha=1, ConstraintWeights_0=1) {
    alphaBoolBlend {blendOption=Linear}
    boneToModify {boneName=dyn_jacket_fr_r_01}
    constraintSetup [{transformType=Rotation, targetBone {boneName=thigh_r}}]
    constraintWeights [1]
   }
   n20 = "Constraint - dyn_jacket_fr_r_02 (R)"(ComponentPose=@n19, Alpha=1, ConstraintWeights_0=1) {
    alphaBoolBlend {blendOption=Linear}
    boneToModify {boneName=dyn_jacket_fr_r_02}
    constraintSetup [{transformType=Rotation, targetBone {boneName=thigh_r}}]
    constraintWeights [1]
   }
   n21 = "Constraint - dyn_jacket_r_MSTR (R)"(ComponentPose=@n20, Alpha=1, ConstraintWeights_0=1) {
    alphaBoolBlend {blendOption=Linear}
    boneToModify {boneName=dyn_jacket_r_MSTR}
    constraintSetup [{transformType=Rotation, targetBone {boneName=thigh_r}}]
    constraintWeights [1]
   }
   n22 = "Constraint - dyn_jacket_r_01 (R)"(ComponentPose=@n21, Alpha=1, ConstraintWeights_0=1) {
    alphaBoolBlend {blendOption=Linear}
    boneToModify {boneName=dyn_jacket_r_01}
    constraintSetup [{transformType=Rotation, targetBone {boneName=thigh_r}}]
    constraintWeights [1]
   }
   n23 = "Constraint - dyn_jacket_r_02 (R)"(ComponentPose=@n22, Alpha=1, ConstraintWeights_0=1) {
    alphaBoolBlend {blendOption=Linear}
    boneToModify {boneName=dyn_jacket_r_02}
    constraintSetup [{transformType=Rotation, targetBone {boneName=thigh_r}}]
    constraintWeights [1]
   }
   n24 = "Transform (Modify) Bone - Bone: dyn_jacket_fr_r_MSTR"!(ComponentPose=@n23, Alpha:float=1, Translation="0,0,2", Rotation="0,0,0", Scale="1,1,1") -> (Pose) {
    rotationMode=BMM_Additive
    rotationSpace=BCS_BoneSpace
    translationMode=BMM_Additive
    translationSpace=BCS_BoneSpace
    alphaBoolBlend {blendOption=Linear}
    boneToModify {boneName=dyn_jacket_fr_r_MSTR}
    translation {z=2}
   }
   n25 = "Transform (Modify) Bone - Bone: dyn_jacket_r_MSTR"!(ComponentPose=@n24, Alpha:float=1, Translation="0,0,2", Rotation="0,0,0", Scale="1,1,1") -> (Pose) {
    rotationMode=BMM_Additive
    rotationSpace=BCS_BoneSpace
    translationMode=BMM_Additive
    translationSpace=BCS_BoneSpace
    alphaBoolBlend {blendOption=Linear}
    boneToModify {boneName=dyn_jacket_r_MSTR}
    translation {z=2}
   }
   n26 = "Transform (Modify) Bone - Bone: Dyn_Shirt_3"(ComponentPose=@n25, Alpha=1, Translation="0,-1,0", Rotation="0,0,0", Scale="1,1,1") {
    translationMode=BMM_Additive
    translationSpace=BCS_BoneSpace
    alphaBoolBlend {blendOption=Linear}
    boneToModify {boneName=Dyn_Shirt_3}
    translation {y=-1}
   }
   n27 = "Constraint - dyn_jacket_fr_r_MSTR (R)"(ComponentPose=@n26, Alpha=1, ConstraintWeights_0=1) {
    alphaBoolBlend {blendOption=Linear}
    alphaScaleBiasClamp {bInterpResult=true}
    boneToModify {boneName=dyn_jacket_fr_r_MSTR}
    constraintSetup [{transformType=Rotation, targetBone {boneName=thigh_r}}]
    constraintWeights [1]
   }
   n28 = "Constraint - dyn_jacket_fr_r_01 (R)"(ComponentPose=@n27, Alpha=1, ConstraintWeights_0=1) {
    alphaBoolBlend {blendOption=Linear}
    alphaScaleBiasClamp {bInterpResult=true}
    boneToModify {boneName=dyn_jacket_fr_r_01}
    constraintSetup [{transformType=Rotation, targetBone {boneName=thigh_r}}]
    constraintWeights [1]
   }
   n29 = "Constraint - dyn_jacket_fr_r_02 (R)"(ComponentPose=@n28, Alpha=1, ConstraintWeights_0=1) {
    alphaBoolBlend {blendOption=Linear}
    alphaScaleBiasClamp {bInterpResult=true}
    boneToModify {boneName=dyn_jacket_fr_r_02}
    constraintSetup [{transformType=Rotation, targetBone {boneName=thigh_r}}]
    constraintWeights [1]
   }
   n30 = "Constraint - dyn_jacket_r_MSTR (R)"(ComponentPose=@n29, Alpha=1, ConstraintWeights_0=1) {
    alphaBoolBlend {blendOption=Linear}
    alphaScaleBiasClamp {bInterpResult=true}
    boneToModify {boneName=dyn_jacket_r_MSTR}
    constraintSetup [{transformType=Rotation, targetBone {boneName=thigh_r}}]
    constraintWeights [1]
   }
   n31 = "Constraint - dyn_jacket_r_01 (R)"(ComponentPose=@n30, Alpha=1, ConstraintWeights_0=1) {
    alphaBoolBlend {blendOption=Linear}
    alphaScaleBiasClamp {bInterpResult=true}
    boneToModify {boneName=dyn_jacket_r_01}
    constraintSetup [{transformType=Rotation, targetBone {boneName=thigh_r}}]
    constraintWeights [1]
   }
   n32 = "Constraint - dyn_jacket_r_02 (R)"(ComponentPose=@n31, Alpha=1, ConstraintWeights_0=1) {
    alphaBoolBlend {blendOption=Linear}
    alphaScaleBiasClamp {bInterpResult=true}
    boneToModify {boneName=dyn_jacket_r_02}
    constraintSetup [{transformType=Rotation, targetBone {boneName=thigh_r}}]
    constraintWeights [1]
   }
   n33 = "Transform (Modify) Bone - Bone: dyn_jacket_fr_r_MSTR"(ComponentPose=@n32, Alpha=1, Translation="0,0,2.5", Rotation="0,0,0", Scale="1,1,1") {
    rotationMode=BMM_Additive
    rotationSpace=BCS_BoneSpace
    translationMode=BMM_Additive
    translationSpace=BCS_BoneSpace
    alphaBoolBlend {blendOption=Linear}
    alphaScaleBiasClamp {bInterpResult=true}
    boneToModify {boneName=dyn_jacket_fr_r_MSTR}
    translation {z=2.5}
   }
   n34 = "Transform (Modify) Bone - Bone: dyn_jacket_r_MSTR"(ComponentPose=@n33, Alpha=1, Translation="0,0,3.5", Rotation="0,0,0", Scale="1,1,1") {
    rotationMode=BMM_Additive
    rotationSpace=BCS_BoneSpace
    translationMode=BMM_Additive
    translationSpace=BCS_BoneSpace
    alphaBoolBlend {blendOption=Linear}
    alphaScaleBiasClamp {bInterpResult=true}
    boneToModify {boneName=dyn_jacket_r_MSTR}
    translation {z=3.5}
   }
   n35 = "Transform (Modify) Bone - Bone: dyn_jacket_bk_r_02"(ComponentPose=@n34, Alpha=Get PelvisDive() -> (PelvisDive:float), Translation="0,0,0", Rotation="0,0,45", Scale="1,1,1") {
    rotationMode=BMM_Additive
    alphaBoolBlend {blendOption=Linear}
    alphaScaleBiasClamp {bInterpResult=true, bMapRange=true, interpSpeedDecreasing=20, interpSpeedIncreasing=20, inRange {max=-30}}
    boneToModify {boneName=dyn_jacket_bk_r_02}
    rotation {roll=45}
   }
   n36 = Get JumpFall() -> (JumpFall:bool)
   n37 = "Transform (Modify) Bone - Bone: dyn_jacket_l_01"!(ComponentPose=@n35, bAlphaBoolEnabled:bool=@n36, Translation="5,-5,0", Rotation="45,0,0", Scale="1,1,1") -> (Pose) {
    alphaInputType=Bool
    rotationMode=BMM_Additive
    translationMode=BMM_Additive
    alphaBoolBlend {blendOption=Linear}
    boneToModify {boneName=dyn_jacket_l_01}
    rotation {pitch=45}
    translation=<5 -5>
   }
   n38 = "Transform (Modify) Bone - Bone: dyn_jacket_bk_l_01"!(ComponentPose=@n37, bAlphaBoolEnabled:bool=@n36, Translation="0,-5,0", Rotation="0,0,45", Scale="1,1,1") -> (Pose) {
    alphaInputType=Bool
    rotationMode=BMM_Additive
    translationMode=BMM_Additive
    alphaBoolBlend {blendOption=Linear}
    boneToModify {boneName=dyn_jacket_bk_l_01}
    rotation {roll=45}
    translation {y=-5}
   }
   n39 = "Transform (Modify) Bone - Bone: dyn_jacket_fr_l_01"!(ComponentPose=@n38, bAlphaBoolEnabled:bool=True, Translation="0,0,0", Rotation="0,0,-10", Scale="1,1,1") -> (Pose) {
    alphaInputType=Bool
    rotationMode=BMM_Additive
    translationMode=BMM_Additive
    alphaBoolBlend {blendOption=Linear}
    boneToModify {boneName=dyn_jacket_fr_l_01}
    rotation {roll=-10}
   }
   n40 = "Transform (Modify) Bone - Bone: dyn_jacket_fr_l_01"(ComponentPose=@n39, bAlphaBoolEnabled=True, Translation="0,0,0", Rotation="0,0,-15", Scale="1,1,1") {
    alphaInputType=Bool
    rotationMode=BMM_Additive
    translationMode=BMM_Additive
    alphaBoolBlend {blendOption=Linear}
    boneToModify {boneName=dyn_jacket_fr_l_01}
    rotation {roll=-15}
   }
   n41 = "Transform (Modify) Bone - Bone: dyn_jacket_bk_l_01"(
    ComponentPose=@n40
    bAlphaBoolEnabled=Get BakcwardRightMoving() -> (BakcwardRightMoving:bool)
    Translation="-5,0,15"
    Rotation="0,0,0"
    Scale="1,1,1"
   ) {
    alphaInputType=Bool
    translationMode=BMM_Additive
    translationSpace=BCS_BoneSpace
    alphaBoolBlend {blendOption=Linear}
    boneToModify {boneName=dyn_jacket_bk_l_01}
    translation {x=-5, z=15}
   }
   n42 = Component To Local(ComponentPose=@n41)
   n43 = "Transform (Modify) Bone - Bone: dyn_jacket_fr_l_MSTR"(
    ComponentPose=Local To Component(LocalPose=Use cached pose 'Crouch'())
    Alpha=1
    Translation="-8.5,-10,0.5"
    Rotation="0,20,24"
    Scale="1,1,1"
   ) {
    rotationMode=BMM_Additive
    rotationSpace=BCS_BoneSpace
    scaleSpace=BCS_BoneSpace
    translationMode=BMM_Additive
    translationSpace=BCS_BoneSpace
    alphaBoolBlend {blendInTime=0.2, blendOption=Linear, blendOutTime=0.2}
    alphaScaleBiasClamp {bClampResult=true, bInterpResult=true, interpSpeedDecreasing=20, interpSpeedIncreasing=20}
    boneToModify {boneName=dyn_jacket_fr_l_MSTR}
    rotation {roll=24, yaw=20}
    translation=<-8.5 -10 0.5>
   }
   n44 = "Transform (Modify) Bone - Bone: dyn_jacket_fr_l_01"!(ComponentPose=@n43, Alpha:float=1, Translation="0,0,0", Rotation="0,0,0", Scale="1,1,1") -> (Pose) {
    rotationMode=BMM_Additive
    rotationSpace=BCS_BoneSpace
    scaleMode=BMM_Additive
    scaleSpace=BCS_BoneSpace
    translationMode=BMM_Additive
    translationSpace=BCS_BoneSpace
    alphaBoolBlend {blendInTime=0.2, blendOption=Linear, blendOutTime=0.2}
    alphaScaleBiasClamp {bClampResult=true, bInterpResult=true, interpSpeedDecreasing=20, interpSpeedIncreasing=20}
    boneToModify {boneName=dyn_jacket_fr_l_01}
   }
   n45 = "Transform (Modify) Bone - Bone: dyn_jacket_l_MSTR"(ComponentPose=@n44, Alpha=1, Translation="-9.5,0,5", Rotation="0,50,0", Scale="1,1,1") {
    rotationMode=BMM_Additive
    rotationSpace=BCS_BoneSpace
    scaleMode=BMM_Additive
    scaleSpace=BCS_BoneSpace
    translationMode=BMM_Additive
    translationSpace=BCS_BoneSpace
    alphaBoolBlend {blendInTime=0.2, blendOption=Linear, blendOutTime=0.2}
    alphaScaleBiasClamp {bInterpResult=true}
    boneToModify {boneName=dyn_jacket_l_MSTR}
    rotation {yaw=50}
    translation {x=-9.5, z=5}
   }
   n46 = Get bIsCrouching()
   n47 = "Transform (Modify) Bone - Bone: dyn_jacket_fr_r_MSTR"!(ComponentPose=@n45, bAlphaBoolEnabled:bool=@n46, Translation="0,0,0", Rotation="0,0,0", Scale="1,1,1") -> (Pose) {
    alphaInputType=Bool
    rotationMode=BMM_Additive
    rotationSpace=BCS_BoneSpace
    scaleMode=BMM_Additive
    scaleSpace=BCS_BoneSpace
    translationMode=BMM_Additive
    translationSpace=BCS_BoneSpace
    alphaBoolBlend {blendInTime=0.2, blendOption=Linear, blendOutTime=0.2}
    alphaScaleBiasClamp {bClampResult=true, bInterpResult=true}
    boneToModify {boneName=dyn_jacket_fr_r_MSTR}
   }
   n48 = "Transform (Modify) Bone - Bone: dyn_jacket_r_MSTR"!(ComponentPose=@n47, bAlphaBoolEnabled:bool=@n46, Translation="0,0,0", Rotation="20,0,10", Scale="1,1,1") -> (Pose) {
    alphaInputType=Bool
    rotationMode=BMM_Additive
    rotationSpace=BCS_BoneSpace
    scaleMode=BMM_Replace
    scaleSpace=BCS_BoneSpace
    translationMode=BMM_Additive
    translationSpace=BCS_BoneSpace
    alphaBoolBlend {blendInTime=0.2, blendOption=Linear, blendOutTime=0.2}
    alphaScaleBiasClamp {bClampResult=true, bInterpResult=true}
    boneToModify {boneName=dyn_jacket_r_MSTR}
    rotation {pitch=20, roll=10}
   }
   n49 = "Transform (Modify) Bone - Bone: dyn_jacket_bk_l_MSTR"!(ComponentPose=@n48, bAlphaBoolEnabled:bool=@n46, Translation="0,0,0", Rotation="0,0,0", Scale="1,1,1") -> (Pose) {
    alphaInputType=Bool
    rotationMode=BMM_Additive
    rotationSpace=BCS_BoneSpace
    scaleSpace=BCS_BoneSpace
    translationMode=BMM_Additive
    translationSpace=BCS_BoneSpace
    alphaBoolBlend {blendInTime=0.2, blendOption=Linear, blendOutTime=0.2}
    alphaScaleBiasClamp {bClampResult=true, bInterpResult=true}
    boneToModify {boneName=dyn_jacket_bk_l_MSTR}
   }
   n50 = Get legUp_r()
   n51 = "Transform (Modify) Bone - Bone: dyn_jacket_fr_r_MSTR"!(ComponentPose=@n49, Alpha:float=@n50, Translation="-7,-5,-2", Rotation="50,0,0", Scale="1,1,1") -> (Pose) {
    rotationMode=BMM_Additive
    rotationSpace=BCS_BoneSpace
    scaleSpace=BCS_BoneSpace
    translationMode=BMM_Additive
    translationSpace=BCS_BoneSpace
    alphaBoolBlend {blendInTime=0.2, blendOption=Linear, blendOutTime=0.2}
    boneToModify {boneName=dyn_jacket_fr_r_MSTR}
    rotation {pitch=50}
    translation=<-7 -5 -2>
   }
   n52 = "Transform (Modify) Bone - Bone: dyn_jacket_r_MSTR"!(ComponentPose=@n51, Alpha:float=@n50, Translation="0,0,3.5", Rotation="35,0,0", Scale="1,1,1") -> (Pose) {
    rotationMode=BMM_Additive
    rotationSpace=BCS_BoneSpace
    scaleSpace=BCS_BoneSpace
    translationMode=BMM_Additive
    translationSpace=BCS_BoneSpace
    alphaBoolBlend {blendInTime=0.2, blendOption=Linear, blendOutTime=0.2}
    alphaScaleBiasClamp {scale=0.5}
    boneToModify {boneName=dyn_jacket_r_MSTR}
    rotation {pitch=35}
    translation {z=3.5}
   }
   n53 = RigidBody!(ComponentPose=@n52, Alpha, ExternalForce) -> (Pose) {
    bClampLinearTranslationLimitToRefPose=true
    bFreezeIncomingPoseOnStart=true
    bOverrideWorldGravity=true
    overridePhysicsAsset=&/M_MED_Hip_Hare_Crouch_Physics
    simulationSpace=BaseBoneSpace
    alphaBoolBlend {blendOption=Linear}
    baseBoneRef {boneName=spine_01}
    componentAppliedLinearAccClamp=<1200 1200 1200>
    componentLinearAccScale=<0.1 0.1 0.1>
    componentLinearVelScale=<0.1 0.1 0.1>
    overrideWorldGravity {z=-400}
   }
   n54 = "Transform (Modify) Bone - Bone: dyn_jacket_fr_r_MSTR"(ComponentPose=@n53, Alpha=1, Translation="0,0,1", Rotation="0,0,0", Scale="1,1,1") {
    translationMode=BMM_Additive
    translationSpace=BCS_BoneSpace
    alphaBoolBlend {blendOption=Linear}
    boneToModify {boneName=dyn_jacket_fr_r_MSTR}
    translation {z=1}
   }
   n55 = "Constraint - dyn_jacket_fr_r_MSTR (R)"(ComponentPose=@n54, Alpha=1, ConstraintWeights_0=1) {
    alphaBoolBlend {blendOption=Linear}
    boneToModify {boneName=dyn_jacket_fr_r_MSTR}
    constraintSetup [{transformType=Rotation, targetBone {boneName=thigh_r}}]
    constraintWeights [1]
   }
   n56 = "Constraint - dyn_jacket_fr_r_01 (R)"(ComponentPose=@n55, Alpha=1, ConstraintWeights_0=1) {
    alphaBoolBlend {blendOption=Linear}
    boneToModify {boneName=dyn_jacket_fr_r_01}
    constraintSetup [{transformType=Rotation, targetBone {boneName=thigh_r}}]
    constraintWeights [1]
   }
   n57 = "Constraint - dyn_jacket_fr_r_02 (R)"(ComponentPose=@n56, Alpha=1, ConstraintWeights_0=1) {
    alphaBoolBlend {blendOption=Linear}
    boneToModify {boneName=dyn_jacket_fr_r_02}
    constraintSetup [{transformType=Rotation, targetBone {boneName=thigh_r}}]
    constraintWeights [1]
   }
   n58 = "Constraint - dyn_jacket_r_MSTR (R)"(ComponentPose=@n57, Alpha=1, ConstraintWeights_0=1) {
    alphaBoolBlend {blendOption=Linear}
    boneToModify {boneName=dyn_jacket_r_MSTR}
    constraintSetup [{transformType=Rotation, targetBone {boneName=thigh_r}}]
    constraintWeights [1]
   }
   n59 = "Constraint - dyn_jacket_r_01 (R)"(ComponentPose=@n58, Alpha=1, ConstraintWeights_0=1) {
    alphaBoolBlend {blendOption=Linear}
    boneToModify {boneName=dyn_jacket_r_01}
    constraintSetup [{transformType=Rotation, targetBone {boneName=thigh_r}}]
    constraintWeights [1]
   }
   n60 = "Constraint - dyn_jacket_r_02 (R)"(ComponentPose=@n59, Alpha=1, ConstraintWeights_0=1) {
    alphaBoolBlend {blendOption=Linear}
    boneToModify {boneName=dyn_jacket_r_02}
    constraintSetup [{transformType=Rotation, targetBone {boneName=thigh_r}}]
    constraintWeights [1]
   }
   n61 = "Transform (Modify) Bone - Bone: dyn_jacket_fr_r_MSTR"(ComponentPose=@n60, Alpha=1, Translation="0,0,3", Rotation="-15,0,-30", Scale="1,1,1") {
    rotationMode=BMM_Additive
    rotationSpace=BCS_BoneSpace
    translationMode=BMM_Additive
    translationSpace=BCS_BoneSpace
    alphaBoolBlend {blendOption=Linear}
    alphaScaleBiasClamp {bInterpResult=true}
    boneToModify {boneName=dyn_jacket_fr_r_MSTR}
    rotation {pitch=-15, roll=-30}
    translation {z=3}
   }
   n62 = "Transform (Modify) Bone - Bone: dyn_jacket_r_MSTR"(ComponentPose=@n61, Alpha=1, Translation="0,0,3", Rotation="15,0,0", Scale="1,1,1") {
    rotationMode=BMM_Additive
    rotationSpace=BCS_BoneSpace
    translationMode=BMM_Additive
    translationSpace=BCS_BoneSpace
    alphaBoolBlend {blendOption=Linear}
    alphaScaleBiasClamp {bInterpResult=true}
    boneToModify {boneName=dyn_jacket_r_MSTR}
    rotation {pitch=15}
    translation {z=3}
   }
   n63 = "Constraint - dyn_jacket_fr_r_MSTR (R)"(ComponentPose=@n62, Alpha=1, ConstraintWeights_0=1) {
    alphaBoolBlend {blendOption=Linear}
    boneToModify {boneName=dyn_jacket_fr_r_MSTR}
    constraintSetup [{transformType=Rotation, targetBone {boneName=thigh_r}}]
    constraintWeights [1]
   }
   n64 = "Constraint - dyn_jacket_fr_r_01 (R)"(ComponentPose=@n63, Alpha=1, ConstraintWeights_0=1) {
    alphaBoolBlend {blendOption=Linear}
    boneToModify {boneName=dyn_jacket_fr_r_01}
    constraintSetup [{transformType=Rotation, targetBone {boneName=thigh_r}}]
    constraintWeights [1]
   }
   n65 = "Constraint - dyn_jacket_fr_r_02 (R)"(ComponentPose=@n64, Alpha=1, ConstraintWeights_0=1) {
    alphaBoolBlend {blendOption=Linear}
    boneToModify {boneName=dyn_jacket_fr_r_02}
    constraintSetup [{transformType=Rotation, targetBone {boneName=thigh_r}}]
    constraintWeights [1]
   }
   n66 = "Constraint - dyn_jacket_r_MSTR (R)"(ComponentPose=@n65, Alpha=1, ConstraintWeights_0=1) {
    alphaBoolBlend {blendOption=Linear}
    boneToModify {boneName=dyn_jacket_r_MSTR}
    constraintSetup [{transformType=Rotation, targetBone {boneName=thigh_r}}]
    constraintWeights [1]
   }
   n67 = "Constraint - dyn_jacket_r_01 (R)"(ComponentPose=@n66, Alpha=1, ConstraintWeights_0=1) {
    alphaBoolBlend {blendOption=Linear}
    boneToModify {boneName=dyn_jacket_r_01}
    constraintSetup [{transformType=Rotation, targetBone {boneName=thigh_r}}]
    constraintWeights [1]
   }
   n68 = "Constraint - dyn_jacket_r_02 (R)"(ComponentPose=@n67, Alpha=1, ConstraintWeights_0=1) {
    alphaBoolBlend {blendOption=Linear}
    boneToModify {boneName=dyn_jacket_r_02}
    constraintSetup [{transformType=Rotation, targetBone {boneName=thigh_r}}]
    constraintWeights [1]
   }
   n69 = "Transform (Modify) Bone - Bone: dyn_jacket_fr_r_MSTR"(ComponentPose=@n68, Alpha=1, Translation="-7,0,4", Rotation="-20,0,-30", Scale="1,1,1") {
    rotationMode=BMM_Additive
    rotationSpace=BCS_BoneSpace
    translationMode=BMM_Additive
    translationSpace=BCS_BoneSpace
    alphaBoolBlend {blendOption=Linear}
    alphaScaleBiasClamp {bInterpResult=true}
    boneToModify {boneName=dyn_jacket_fr_r_MSTR}
    rotation {pitch=-20, roll=-30}
    translation {x=-7, z=4}
   }
   n70 = "Transform (Modify) Bone - Bone: dyn_jacket_r_MSTR"(ComponentPose=@n69, Alpha=1, Translation="-7,0,4", Rotation="15,0,0", Scale="1,1,1") {
    rotationMode=BMM_Additive
    rotationSpace=BCS_BoneSpace
    translationMode=BMM_Additive
    translationSpace=BCS_BoneSpace
    alphaBoolBlend {blendOption=Linear}
    alphaScaleBiasClamp {bInterpResult=true}
    boneToModify {boneName=dyn_jacket_r_MSTR}
    rotation {pitch=15}
    translation {x=-7, z=4}
   }
   n71 = "Constraint - dyn_jacket_fr_l_MSTR (R)"(ComponentPose=@n70, Alpha=1, ConstraintWeights_0=1) {
    alphaBoolBlend {blendOption=Linear}
    boneToModify {boneName=dyn_jacket_fr_l_MSTR}
    constraintSetup [{transformType=Rotation, targetBone {boneName=thigh_l}}]
    constraintWeights [1]
   }
   n72 = "Constraint - dyn_jacket_fr_l_01 (R)"(ComponentPose=@n71, Alpha=1, ConstraintWeights_0=1) {
    alphaBoolBlend {blendOption=Linear}
    boneToModify {boneName=dyn_jacket_fr_l_01}
    constraintSetup [{transformType=Rotation, targetBone {boneName=thigh_l}}]
    constraintWeights [1]
   }
   n73 = "Constraint - dyn_jacket_fr_l_02 (R)"(ComponentPose=@n72, Alpha=1, ConstraintWeights_0=1) {
    alphaBoolBlend {blendOption=Linear}
    boneToModify {boneName=dyn_jacket_fr_l_02}
    constraintSetup [{transformType=Rotation, targetBone {boneName=thigh_l}}]
    constraintWeights [1]
   }
   n74 = "Transform (Modify) Bone - Bone: dyn_jacket_fr_l_MSTR"(ComponentPose=@n73, Alpha=1, Translation="0,0,0", Rotation="-15,0,0", Scale="1,1,1") {
    rotationMode=BMM_Additive
    rotationSpace=BCS_BoneSpace
    translationMode=BMM_Additive
    translationSpace=BCS_BoneSpace
    alphaBoolBlend {blendOption=Linear}
    alphaScaleBiasClamp {bInterpResult=true}
    boneToModify {boneName=dyn_jacket_fr_l_MSTR}
    rotation {pitch=-15}
   }
   n75 = "Constraint - dyn_jacket_fr_r_MSTR (R)"(ComponentPose=@n74, Alpha=1, ConstraintWeights_0=1) {
    alphaBoolBlend {blendOption=Linear}
    boneToModify {boneName=dyn_jacket_fr_r_MSTR}
    constraintSetup [{transformType=Rotation, targetBone {boneName=thigh_r}}]
    constraintWeights [1]
   }
   n76 = "Constraint - dyn_jacket_fr_r_01 (R)"(ComponentPose=@n75, Alpha=1, ConstraintWeights_0=1) {
    alphaBoolBlend {blendOption=Linear}
    boneToModify {boneName=dyn_jacket_fr_r_01}
    constraintSetup [{transformType=Rotation, targetBone {boneName=thigh_r}}]
    constraintWeights [1]
   }
   n77 = "Constraint - dyn_jacket_fr_r_02 (R)"(ComponentPose=@n76, Alpha=1, ConstraintWeights_0=1) {
    alphaBoolBlend {blendOption=Linear}
    boneToModify {boneName=dyn_jacket_fr_r_02}
    constraintSetup [{transformType=Rotation, targetBone {boneName=thigh_r}}]
    constraintWeights [1]
   }
   n78 = "Constraint - dyn_jacket_r_MSTR (R)"(ComponentPose=@n77, Alpha=1, ConstraintWeights_0=1) {
    alphaBoolBlend {blendOption=Linear}
    boneToModify {boneName=dyn_jacket_r_MSTR}
    constraintSetup [{transformType=Rotation, targetBone {boneName=thigh_r}}]
    constraintWeights [1]
   }
   n79 = "Constraint - dyn_jacket_r_01 (R)"(ComponentPose=@n78, Alpha=1, ConstraintWeights_0=1) {
    alphaBoolBlend {blendOption=Linear}
    boneToModify {boneName=dyn_jacket_r_01}
    constraintSetup [{transformType=Rotation, targetBone {boneName=thigh_r}}]
    constraintWeights [1]
   }
   n80 = "Constraint - dyn_jacket_r_02 (R)"(ComponentPose=@n79, Alpha=1, ConstraintWeights_0=1) {
    alphaBoolBlend {blendOption=Linear}
    boneToModify {boneName=dyn_jacket_r_02}
    constraintSetup [{transformType=Rotation, targetBone {boneName=thigh_r}}]
    constraintWeights [1]
   }
   n81 = "Transform (Modify) Bone - Bone: dyn_jacket_fr_r_MSTR"(ComponentPose=@n80, Alpha=1, Translation="-5,0,5", Rotation="-15,0,0", Scale="1,1,1") {
    rotationMode=BMM_Additive
    rotationSpace=BCS_BoneSpace
    translationMode=BMM_Additive
    translationSpace=BCS_BoneSpace
    alphaBoolBlend {blendOption=Linear}
    alphaScaleBiasClamp {bInterpResult=true}
    boneToModify {boneName=dyn_jacket_fr_r_MSTR}
    rotation {pitch=-15}
    translation {x=-5, z=5}
   }
   n82 = "Transform (Modify) Bone - Bone: dyn_jacket_r_MSTR"(ComponentPose=@n81, Alpha=1, Translation="0,0,5", Rotation="0,0,0", Scale="1,1,1") {
    rotationMode=BMM_Additive
    rotationSpace=BCS_BoneSpace
    translationMode=BMM_Additive
    translationSpace=BCS_BoneSpace
    alphaBoolBlend {blendOption=Linear}
    alphaScaleBiasClamp {bInterpResult=true}
    boneToModify {boneName=dyn_jacket_r_MSTR}
    translation {z=5}
   }
   n83 = "Constraint - dyn_jacket_fr_l_MSTR (P)"(ComponentPose=@n82, Alpha=1, ConstraintWeights_0=1) {
    alphaBoolBlend {blendOption=Linear}
    boneToModify {boneName=dyn_jacket_fr_l_MSTR}
    constraintSetup [{targetBone {boneName=pelvis}}]
    constraintWeights [1]
   }
   n84 = "Constraint - dyn_jacket_fr_l_MSTR (R)"(ComponentPose=@n83, Alpha=1, ConstraintWeights_0=1) {
    alphaBoolBlend {blendOption=Linear}
    boneToModify {boneName=dyn_jacket_fr_l_MSTR}
    constraintSetup [{transformType=Rotation, targetBone {boneName=thigh_l}}]
    constraintWeights [1]
   }
   n85 = "Constraint - dyn_jacket_fr_l_01 (R)"(ComponentPose=@n84, Alpha=1, ConstraintWeights_0=1) {
    alphaBoolBlend {blendOption=Linear}
    boneToModify {boneName=dyn_jacket_fr_l_01}
    constraintSetup [{transformType=Rotation, targetBone {boneName=thigh_l}}]
    constraintWeights [1]
   }
   n86 = "Constraint - dyn_jacket_fr_l_02 (R)"(ComponentPose=@n85, Alpha=1, ConstraintWeights_0=1) {
    alphaBoolBlend {blendOption=Linear}
    boneToModify {boneName=dyn_jacket_fr_l_02}
    constraintSetup [{transformType=Rotation, targetBone {boneName=thigh_l}}]
    constraintWeights [1]
   }
   n87 = "Constraint - dyn_jacket_l_MSTR (P)"(ComponentPose=@n86, Alpha=1, ConstraintWeights_0=1) {
    alphaBoolBlend {blendOption=Linear}
    boneToModify {boneName=dyn_jacket_l_MSTR}
    constraintSetup [{targetBone {boneName=pelvis}}]
    constraintWeights [1]
   }
   n88 = "Constraint - dyn_jacket_l_MSTR (R)"(ComponentPose=@n87, Alpha=1, ConstraintWeights_0=1) {
    alphaBoolBlend {blendOption=Linear}
    boneToModify {boneName=dyn_jacket_l_MSTR}
    constraintSetup [{transformType=Rotation, targetBone {boneName=thigh_l}}]
    constraintWeights [1]
   }
   n89 = "Constraint - dyn_jacket_l_01 (R)"(ComponentPose=@n88, Alpha=1, ConstraintWeights_0=1) {
    alphaBoolBlend {blendOption=Linear}
    boneToModify {boneName=dyn_jacket_l_01}
    constraintSetup [{transformType=Rotation, targetBone {boneName=thigh_l}}]
    constraintWeights [1]
   }
   n90 = "Constraint - dyn_jacket_l_02 (R)"(ComponentPose=@n89, Alpha=1, ConstraintWeights_0=1) {
    alphaBoolBlend {blendOption=Linear}
    boneToModify {boneName=dyn_jacket_l_02}
    constraintSetup [{transformType=Rotation, targetBone {boneName=thigh_l}}]
    constraintWeights [1]
   }
   n91 = "Transform (Modify) Bone - Bone: dyn_jacket_fr_l_MSTR"(ComponentPose=@n90, Alpha=1, Translation="-5,0,10", Rotation="-15,0,0", Scale="1,1,1") {
    rotationMode=BMM_Additive
    rotationSpace=BCS_BoneSpace
    translationMode=BMM_Additive
    translationSpace=BCS_BoneSpace
    alphaBoolBlend {blendOption=Linear}
    alphaScaleBiasClamp {bInterpResult=true}
    boneToModify {boneName=dyn_jacket_fr_l_MSTR}
    rotation {pitch=-15}
    translation {x=-5, z=10}
   }
   n92 = "Transform (Modify) Bone - Bone: dyn_jacket_l_MSTR"(ComponentPose=@n91, Alpha=1, Translation="0,0,5", Rotation="0,0,0", Scale="1,1,1") {
    rotationMode=BMM_Additive
    rotationSpace=BCS_BoneSpace
    translationMode=BMM_Additive
    translationSpace=BCS_BoneSpace
    alphaBoolBlend {blendOption=Linear}
    alphaScaleBiasClamp {bInterpResult=true}
    boneToModify {boneName=dyn_jacket_l_MSTR}
    translation {z=5}
   }
   n93 = "Constraint - dyn_jacket_fr_r_MSTR (R)"(ComponentPose=@n92, Alpha=1, ConstraintWeights_0=1) {
    alphaBoolBlend {blendOption=Linear}
    boneToModify {boneName=dyn_jacket_fr_r_MSTR}
    constraintSetup [{transformType=Rotation, targetBone {boneName=thigh_r}}]
    constraintWeights [1]
   }
   n94 = "Constraint - dyn_jacket_fr_r_01 (R)"(ComponentPose=@n93, Alpha=1, ConstraintWeights_0=1) {
    alphaBoolBlend {blendOption=Linear}
    boneToModify {boneName=dyn_jacket_fr_r_01}
    constraintSetup [{transformType=Rotation, targetBone {boneName=thigh_r}}]
    constraintWeights [1]
   }
   n95 = "Constraint - dyn_jacket_fr_r_02 (R)"(ComponentPose=@n94, Alpha=1, ConstraintWeights_0=1) {
    alphaBoolBlend {blendOption=Linear}
    boneToModify {boneName=dyn_jacket_fr_r_02}
    constraintSetup [{transformType=Rotation, targetBone {boneName=thigh_r}}]
    constraintWeights [1]
   }
   n96 = "Constraint - dyn_jacket_r_MSTR (R)"(ComponentPose=@n95, Alpha=1, ConstraintWeights_0=1) {
    alphaBoolBlend {blendOption=Linear}
    boneToModify {boneName=dyn_jacket_r_MSTR}
    constraintSetup [{transformType=Rotation, targetBone {boneName=thigh_r}}]
    constraintWeights [1]
   }
   n97 = "Constraint - dyn_jacket_r_01 (R)"(ComponentPose=@n96, Alpha=1, ConstraintWeights_0=1) {
    alphaBoolBlend {blendOption=Linear}
    boneToModify {boneName=dyn_jacket_r_01}
    constraintSetup [{transformType=Rotation, targetBone {boneName=thigh_r}}]
    constraintWeights [1]
   }
   n98 = "Constraint - dyn_jacket_r_02 (R)"(ComponentPose=@n97, Alpha=1, ConstraintWeights_0=1) {
    alphaBoolBlend {blendOption=Linear}
    boneToModify {boneName=dyn_jacket_r_02}
    constraintSetup [{transformType=Rotation, targetBone {boneName=thigh_r}}]
    constraintWeights [1]
   }
   n99 = "Transform (Modify) Bone - Bone: dyn_jacket_fr_r_MSTR"(ComponentPose=@n98, Alpha=1, Translation="-7,1,3", Rotation="-20,0,-45", Scale="1,1,1") {
    rotationMode=BMM_Additive
    rotationSpace=BCS_BoneSpace
    translationMode=BMM_Additive
    translationSpace=BCS_BoneSpace
    alphaBoolBlend {blendOption=Linear}
    alphaScaleBiasClamp {bInterpResult=true}
    boneToModify {boneName=dyn_jacket_fr_r_MSTR}
    rotation {pitch=-20, roll=-45}
    translation=<-7 1 3>
   }
   n100 = "Transform (Modify) Bone - Bone: dyn_jacket_r_MSTR"(ComponentPose=@n99, Alpha=1, Translation="-7,0,3", Rotation="0,0,0", Scale="1,1,1") {
    rotationMode=BMM_Additive
    rotationSpace=BCS_BoneSpace
    translationMode=BMM_Additive
    translationSpace=BCS_BoneSpace
    alphaBoolBlend {blendOption=Linear}
    alphaScaleBiasClamp {bInterpResult=true}
    boneToModify {boneName=dyn_jacket_r_MSTR}
    translation {x=-7, z=3}
   }
   n101 = "Constraint - dyn_jacket_fr_l_MSTR (R)"(ComponentPose=@n100, Alpha=1, ConstraintWeights_0=1) {
    alphaBoolBlend {blendOption=Linear}
    boneToModify {boneName=dyn_jacket_fr_l_MSTR}
    constraintSetup [{transformType=Rotation, targetBone {boneName=thigh_l}}]
    constraintWeights [1]
   }
   n102 = "Constraint - dyn_jacket_fr_l_01 (R)"(ComponentPose=@n101, Alpha=1, ConstraintWeights_0=1) {
    alphaBoolBlend {blendOption=Linear}
    boneToModify {boneName=dyn_jacket_fr_l_01}
    constraintSetup [{transformType=Rotation, targetBone {boneName=thigh_l}}]
    constraintWeights [1]
   }
   n103 = "Constraint - dyn_jacket_fr_l_02 (R)"(ComponentPose=@n102, Alpha=1, ConstraintWeights_0=1) {
    alphaBoolBlend {blendOption=Linear}
    boneToModify {boneName=dyn_jacket_fr_l_02}
    constraintSetup [{transformType=Rotation, targetBone {boneName=thigh_l}}]
    constraintWeights [1]
   }
   n104 = "Transform (Modify) Bone - Bone: dyn_jacket_fr_l_MSTR"(ComponentPose=@n103, Alpha=1, Translation="0,0,0", Rotation="5,-15,0", Scale="1,1,1") {
    rotationMode=BMM_Additive
    rotationSpace=BCS_BoneSpace
    translationMode=BMM_Additive
    translationSpace=BCS_BoneSpace
    alphaBoolBlend {blendOption=Linear}
    alphaScaleBiasClamp {bInterpResult=true}
    boneToModify {boneName=dyn_jacket_fr_l_MSTR}
    rotation {pitch=5, yaw=-15}
   }
   n105 = "Transform (Modify) Bone - Bone: dyn_jacket_l_MSTR"(ComponentPose=@n104, Alpha=1, Translation="0,0,0.5", Rotation="15,0,0", Scale="1,1,1") {
    rotationMode=BMM_Additive
    rotationSpace=BCS_BoneSpace
    translationMode=BMM_Additive
    translationSpace=BCS_BoneSpace
    alphaBoolBlend {blendOption=Linear}
    alphaScaleBiasClamp {bInterpResult=true, interpSpeedDecreasing=20, interpSpeedIncreasing=20}
    boneToModify {boneName=dyn_jacket_l_MSTR}
    rotation {pitch=15}
    translation {z=0.5}
   }
   n106 = "Constraint - dyn_jacket_fr_l_MSTR (P)"(ComponentPose=@n105, Alpha=1, ConstraintWeights_0=1) {
    alphaBoolBlend {blendOption=Linear}
    boneToModify {boneName=dyn_jacket_fr_l_MSTR}
    constraintSetup [{targetBone {boneName=pelvis}}]
    constraintWeights [1]
   }
   n107 = "Constraint - dyn_jacket_fr_l_MSTR (R)"(ComponentPose=@n106, Alpha=1, ConstraintWeights_0=1) {
    alphaBoolBlend {blendOption=Linear}
    boneToModify {boneName=dyn_jacket_fr_l_MSTR}
    constraintSetup [{transformType=Rotation, targetBone {boneName=thigh_l}}]
    constraintWeights [1]
   }
   n108 = "Constraint - dyn_jacket_fr_l_01 (R)"(ComponentPose=@n107, Alpha=1, ConstraintWeights_0=1) {
    alphaBoolBlend {blendOption=Linear}
    boneToModify {boneName=dyn_jacket_fr_l_01}
    constraintSetup [{transformType=Rotation, targetBone {boneName=thigh_l}}]
    constraintWeights [1]
   }
   n109 = "Constraint - dyn_jacket_fr_l_02 (R)"(ComponentPose=@n108, Alpha=1, ConstraintWeights_0=1) {
    alphaBoolBlend {blendOption=Linear}
    boneToModify {boneName=dyn_jacket_fr_l_02}
    constraintSetup [{transformType=Rotation, targetBone {boneName=thigh_l}}]
    constraintWeights [1]
   }
   n110 = "Constraint - dyn_jacket_fr_r_MSTR (P)"(ComponentPose=@n109, Alpha=1, ConstraintWeights_0=1) {
    alphaBoolBlend {blendOption=Linear}
    boneToModify {boneName=dyn_jacket_fr_r_MSTR}
    constraintSetup [{targetBone {boneName=pelvis}}]
    constraintWeights [1]
   }
   n111 = "Constraint - dyn_jacket_fr_r_MSTR (R)"(ComponentPose=@n110, Alpha=1, ConstraintWeights_0=1) {
    alphaBoolBlend {blendOption=Linear}
    boneToModify {boneName=dyn_jacket_fr_r_MSTR}
    constraintSetup [{transformType=Rotation, targetBone {boneName=thigh_r}}]
    constraintWeights [1]
   }
   n112 = "Constraint - dyn_jacket_fr_r_01 (R)"(ComponentPose=@n111, Alpha=1, ConstraintWeights_0=1) {
    alphaBoolBlend {blendOption=Linear}
    boneToModify {boneName=dyn_jacket_fr_r_01}
    constraintSetup [{transformType=Rotation, targetBone {boneName=thigh_r}}]
    constraintWeights [1]
   }
   n113 = "Constraint - dyn_jacket_fr_r_02 (R)"(ComponentPose=@n112, Alpha=1, ConstraintWeights_0=1) {
    alphaBoolBlend {blendOption=Linear}
    boneToModify {boneName=dyn_jacket_fr_r_02}
    constraintSetup [{transformType=Rotation, targetBone {boneName=thigh_r}}]
    constraintWeights [1]
   }
   n114 = "Transform (Modify) Bone - Bone: dyn_jacket_fr_l_MSTR"(ComponentPose=@n113, Alpha=1, Translation="-5,0,10", Rotation="-45,0,0", Scale="1,1,1") {
    rotationMode=BMM_Additive
    rotationSpace=BCS_BoneSpace
    translationMode=BMM_Additive
    translationSpace=BCS_BoneSpace
    alphaBoolBlend {blendOption=Linear}
    alphaScaleBiasClamp {bInterpResult=true}
    boneToModify {boneName=dyn_jacket_fr_l_MSTR}
    rotation {pitch=-45}
    translation {x=-5, z=10}
   }
   n115 = "Transform (Modify) Bone - Bone: dyn_jacket_fr_r_MSTR"(ComponentPose=@n114, Alpha=1, Translation="-5,0,10", Rotation="-15,15,0", Scale="1,1,1") {
    rotationMode=BMM_Additive
    rotationSpace=BCS_BoneSpace
    translationMode=BMM_Additive
    translationSpace=BCS_BoneSpace
    alphaBoolBlend {blendOption=Linear}
    alphaScaleBiasClamp {bInterpResult=true}
    boneToModify {boneName=dyn_jacket_fr_r_MSTR}
    rotation {pitch=-15, yaw=15}
    translation {x=-5, z=10}
   }
   n116 = Get FrontRightCrouchMoving() -> (FrontRightCrouchMoving:bool)
   n117 = "Transform (Modify) Bone - Bone: dyn_jacket_fr_r_MSTR"!(ComponentPose=@n115, bAlphaBoolEnabled:bool=@n116, Translation="0,0,0", Rotation="15,15,0", Scale="1,1,1") -> (Pose) {
    alphaInputType=Bool
    rotationMode=BMM_Additive
    rotationSpace=BCS_BoneSpace
    translationMode=BMM_Additive
    translationSpace=BCS_BoneSpace
    alphaBoolBlend {blendInTime=0.1, blendOption=Linear, blendOutTime=0.1}
    boneToModify {boneName=dyn_jacket_fr_r_MSTR}
    rotation {pitch=15, yaw=15}
   }
   n118 = "Transform (Modify) Bone - Bone: dyn_jacket_r_MSTR"!(ComponentPose=@n117, bAlphaBoolEnabled:bool=@n116, Translation="0,0,0", Rotation="15,-15,0", Scale="1,1,1") -> (Pose) {
    alphaInputType=Bool
    rotationMode=BMM_Additive
    rotationSpace=BCS_BoneSpace
    translationMode=BMM_Additive
    translationSpace=BCS_BoneSpace
    alphaBoolBlend {blendInTime=0.1, blendOption=Linear, blendOutTime=0.1}
    boneToModify {boneName=dyn_jacket_r_MSTR}
    rotation {pitch=15, yaw=-15}
   }
   n119 = Component To Local(ComponentPose=@n118)
   n120 = Get bIsCrouching()
   n121 = "Transform (Modify) Bone - Bone: dyn_jacket_r_MSTR"!(
    ComponentPose=Local To Component(LocalPose=Blend Poses by bool(BlendPose_0=@n119, BlendPose_1=@n42, bActiveValue=@n120) {blendType=Linear})
    Alpha:float=1
    Translation="0,0,5"
    Rotation="0,0,0"
    Scale="1,1,1"
   ) -> (Pose) {
    rotationMode=BMM_Additive
    rotationSpace=BCS_BoneSpace
    translationMode=BMM_Additive
    translationSpace=BCS_BoneSpace
    alphaBoolBlend {blendOption=Linear}
    alphaScaleBiasClamp {bInterpResult=true}
    boneToModify {boneName=dyn_jacket_r_MSTR}
    translation {z=5}
   }
   n122 = "Transform (Modify) Bone - Bone: dyn_jacket_fr_r_MSTR"!(ComponentPose=@n121, Alpha:float=1, Translation="0,0,2", Rotation="0,0,0", Scale="1,1,1") -> (Pose) {
    rotationMode=BMM_Additive
    rotationSpace=BCS_BoneSpace
    translationMode=BMM_Additive
    translationSpace=BCS_BoneSpace
    alphaBoolBlend {blendOption=Linear}
    alphaScaleBiasClamp {bInterpResult=true}
    boneToModify {boneName=dyn_jacket_fr_r_MSTR}
    translation {z=2}
   }
   n123 = "Constraint - Dyn_Hand_l (RR)"!(ComponentPose=@n122, Alpha=1, ConstraintWeights_0=0.9, ConstraintWeights_1:float=0.1) -> (Pose) {
    alphaBoolBlend {blendOption=Linear}
    boneToModify {boneName=Dyn_Hand_l}
    constraintSetup [{transformType=Rotation, targetBone {boneName=hand_l}}, {transformType=Rotation, targetBone {boneName=lowerarm_l}}]
    constraintWeights [1,1]
   }
   n124 = "Constraint - Dyn_Hand_r (RR)"!(ComponentPose=@n123, Alpha=1, ConstraintWeights_0=0.8, ConstraintWeights_1:float=0.2) -> (Pose) {
    alphaBoolBlend {blendOption=Linear}
    boneToModify {boneName=Dyn_Hand_r}
    constraintSetup [{transformType=Rotation, targetBone {boneName=hand_r}}, {transformType=Rotation, targetBone {boneName=lowerarm_r}}]
    constraintWeights [1,1]
   }
   Output Pose(Result=Component To Local(ComponentPose=@n124)) {layerGroup=None}
   Save cached pose 'Crouch'(Pose=Copy Pose From Mesh() {bUseAttachedParent=true}) {cacheName=Crouch}
  ]}
 EventGraph {nodes [
   Delay(Duration=1)
   > ResetDynamics(InTeleportType=TeleportPhysics)
   > ResetDynamics(InTeleportType=ResetPhysics)
   Event BlueprintUpdateAnimation()
   > n1 = Sequence!(execute) -> (then_0, then_1, then_2:Exec, then_3:Exec, then_4:Exec, then_5:Exec)
   > Set crouchWalking(crouchWalking:bool=AND Boolean!(A=Get bIsCrouching(), B=Get bIsMoving2D(), C:bool) -> (ReturnValue)) -> (Output_Get:bool)
   > Set legUp_r(legUp_r:double=GetAlphaFromReferenceTransformDelta(
     SkeletalMeshComponent=Get skelMesh() -> (skelMesh:Skeletal Mesh Component Object Reference)
     BoneName=thigh_r
     BaseName=pelvis
     InputRange=(X=-37,Y=-104)
     TwistAxis="1,0,0"
    )) -> (Output_Get:double)
   Set JumpFall(execute=@n1.then_4, JumpFall:bool=OR Boolean(A=Get bIsFalling(), B=Get bIsJumping())) -> (Output_Get:bool)
   n2 = Get bIsCrouching()
   n3 = Get bIsSliding()
   > Set SprintSlide(SprintSlide:bool=OR Boolean!(A=Get bIsSprinting(), B=AND Boolean!(A=@n3, B=@n2) -> (ReturnValue), C:bool) -> (ReturnValue)) -> (Output_Get:bool)
   n4 = Get PawnMovementDirectionAzimuth() -> (PawnMovementDirectionAzimuth:float)
   n5 = float - float(A=MapRangeClamped(Value=@n4, InRangeA=90, InRangeB=135), B=MapRangeClamped(Value=@n4, InRangeA=135, InRangeB=180)) {is=title K2Node_CallFunction}
   n6 = Get bIsCrouching()
   n7 = Get bIsMoving2D()
   > Set BackRight(BackRight:double=SelectFloat(A="Absolute (Float)"(A=@n5), B=0, bPickA=AND Boolean(A=@n6, B=@n7))) -> (Output_Get:double)
   Set legUp_l(
    execute=@n1.then_1
    legUp_l:double=GetAlphaFromReferenceTransformDelta(SkeletalMeshComponent=Get skelMesh(), BoneName=thigh_l, BaseName=pelvis, InputRange=(X=-37,Y=-104), TwistAxis="1,0,0")
   ) -> (Output_Get:double)
   n8 = Get PawnMovementDirectionAzimuth()
   n9 = float - float(A=MapRangeClamped(Value=@n8, InRangeA=45, InRangeB=90), B=MapRangeClamped(Value=@n8, InRangeA=90, InRangeB=135))
   n10 = Get bIsCrouching()
   n11 = Get bIsMoving2D()
   > Set legR_r(legR_r:double=SelectFloat(A="Absolute (Float)"(A=@n9), B=0, bPickA=AND Boolean(A=@n10, B=@n11))) -> (Output_Get:double)
   n12 = Get PawnMovementDirectionAzimuth()
   n13 = float - float(A=MapRangeClamped(Value=@n12, InRangeA=-90, InRangeB=-135), B=MapRangeClamped(Value=@n12, InRangeA=-135, InRangeB=-180))
   n14 = Get bIsMoving2D()
   n15 = Get bIsCrouching()
   Set BackLeft(execute=@n1.then_2, BackLeft:double=SelectFloat(A="Absolute (Float)"(A=@n13), B=0, bPickA=AND Boolean(A=@n15, B=@n14))) -> (Output_Get:double)
   n16 = Get PawnMovementDirectionAzimuth()
   n17 = float - float(A=MapRangeClamped(Value=@n16, InRangeA=-45, InRangeB=-90), B=MapRangeClamped(Value=@n16, InRangeA=-90, InRangeB=-135))
   n18 = Get bIsCrouching()
   n19 = Get bIsMoving2D()
   > Set crouchLeft(crouchLeft:double=SelectFloat(A="Absolute (Float)"(A=@n17), B=0, bPickA=AND Boolean(A=@n18, B=@n19))) -> (Output_Get:double)
   Set JumpSprint(execute=@n1.then_3, JumpSprint:bool=AND Boolean(A=Get bIsJumping(), B=Get bIsSprinting())) -> (Output_Get:bool)
   n20 = Get PawnMovementDirectionAzimuth()
   n21 = MapRangeClamped(Value=@n20, InRangeA=-135, InRangeB=-180, OutRangeB=0.75)
   n22 = MapRangeClamped(Value=@n20, InRangeA=135, InRangeB=180, OutRangeB=0.75)
   n23 = Get bIsCrouching()
   n24 = Get bIsMoving2D()
   > Set crouchBack(crouchBack:double=SelectFloat(A=float + float(A=@n21, B=@n22) {is=title K2Node_CommutativeAssociativeBinaryOperator}, B=0, bPickA=AND Boolean(A=@n23, B=@n24))) -> (Output_Get:double)
   n25 = Get PawnMovementDirectionAzimuth()
   n26 = MapRangeClamped(Value=@n25, InRangeA=-135, InRangeB=-180, OutRangeB=0.75)
   n27 = MapRangeClamped(Value=@n25, InRangeA=135, InRangeB=180, OutRangeB=0.75)
   n28 = Get bIsMoving2D()
   Set movingBack(execute=@n1.then_5, movingBack:double=SelectFloat(A=float + float(A=@n26, B=@n27), B=0, bPickA=@n28)) -> (Output_Get:double)
   > Set baseballSliding(baseballSliding:bool=false) -> (Output_Get:bool)
   > n29 = Gravity Override State Function DEPRECATED(CharacterPartAnimInstance=Self-Reference(), GameTimeInSeconds=GetGameTimeInSeconds(), GravityOverrideParams=Get Gravity Override Params())
   > Set Gravity Override(Gravity Override=@n29)
   > n30 = Sequence!(execute) -> (then_0, then_1)
   > Set PelvisDive(PelvisDive:float=GetSocketRotation!(self:Scene Component Object Reference=Get skelMesh(), InSocketName:Name=pelvis) -> (ReturnValue_Roll:float, ReturnValue_Pitch:float, ReturnValue_Yaw:float).ReturnValue_Pitch) -> (Output_Get:float)
   n31 = Get bIsCrouching()
   Set Standing Moving(
    execute=@n30.then_1
    Standing Moving:bool=AND Boolean!(A=Get bIsMoving2D(), B=Get bIsOnGround() -> (bIsOnGround:bool), C:bool=NOT Boolean(A=@n31)) -> (ReturnValue)
   ) -> (Output_Get:bool)
   n32 = Get PawnMovementDirectionAzimuth()
   > Set BakcwardRightMoving(BakcwardRightMoving:bool=AND Boolean(A="float > float"(A=@n32, B=120), B="float < float"(A=@n32, B=150), C=Get Standing Moving() -> (Standing Moving:bool))) -> (Output_Get:bool)
   n33 = Get bIsCrouching()
   n34 = Get bIsMoving2D()
   n35 = Get bIsOnGround()
   n36 = Get PawnMovementDirectionAzimuth()
   > Set FrontRightCrouchMoving(FrontRightCrouchMoving:bool=AND Boolean(A="float > float"(A=@n36, B=15), B="float < float"(A=@n36, B=75), C=AND Boolean(A=@n34, B=@n35, C=@n33))) -> (Output_Get:bool)
   Event BlueprintInitializeAnimation()
   > Set skelMesh(skelMesh:Skeletal Mesh Component Object Reference=GetOwnerSkeletalMeshComponent()) -> (Output_Get:Skeletal Mesh Component Object Reference)
   Event OnBeginPlayingEmote() {is=title K2Node_Event}
   > Delay(Duration=1)
   > ResetDynamics(InTeleportType=TeleportPhysics)
   > ResetDynamics(InTeleportType=ResetPhysics)
   n37 = GetAnimInstance(self=GetOwnerSkeletalMeshComponent())
   OR Boolean!(
    A=OR Boolean(
     A="Equal (Object)"(A=@n37, B=/Game/Animation/Game/MainPlayer/Locomotion/Medium/Male/Sliding/KneeSliding/BaseballSlideInitiate)
     B="Equal (Object)"(A=@n37, B=/Game/Animation/Game/MainPlayer/Locomotion/Medium/Male/Sliding/KneeSliding/BaseballSlideInitiate_SpineTweak)
     C="Equal (Object)"(A=@n37, B=/Game/Animation/Game/MainPlayer/Locomotion/Medium/Male/Sliding/KneeSliding/BaseballSliding)
     D="Equal (Object)"(A=@n37, B=/Game/Animation/Game/MainPlayer/Locomotion/Medium/Male/Sliding/KneeSliding/BaseballSlidingSlopeAngle-45E)
     E="Equal (Object)"(A=@n37, B=/Game/Animation/Game/MainPlayer/Locomotion/Medium/Male/Sliding/KneeSliding/BaseballSlidingSlopeAngle-45N)
    )
    B=OR Boolean(
     A="Equal (Object)"(A=@n37, B=/Game/Animation/Game/MainPlayer/Locomotion/Medium/Male/Sliding/KneeSliding/BaseballSlidingSlopeAngle45N)
     B="Equal (Object)"(A=@n37, B=/Game/Animation/Game/MainPlayer/Locomotion/Medium/Male/Sliding/KneeSliding/BaseballSlidingSlopeAngle45S)
     C="Equal (Object)"(A=@n37, B=/Game/Animation/Game/MainPlayer/Locomotion/Medium/Male/Sliding/KneeSliding/BaseballSlidingSlopeAngle45SE)
     D="Equal (Object)"(A=@n37, B=/Game/Animation/Game/MainPlayer/Locomotion/Medium/Male/Sliding/KneeSliding/BaseballSlidingSlopeAngle45SW)
     E="Equal (Object)"(A=@n37, B=/Game/Animation/Game/MainPlayer/Locomotion/Medium/Male/Sliding/KneeSliding/BaseballSlidingSlopeAngle45W)
    )
    C=OR Boolean(
     A="Equal (Object)"(A=@n37, B=/Game/Animation/Game/MainPlayer/Locomotion/Medium/Male/Sliding/KneeSliding/BaseballSlidingSlopeAngle-45S)
     B="Equal (Object)"(A=@n37, B=/Game/Animation/Game/MainPlayer/Locomotion/Medium/Male/Sliding/KneeSliding/BaseballSlidingSlopeAngle-45SE)
     C="Equal (Object)"(A=@n37, B=/Game/Animation/Game/MainPlayer/Locomotion/Medium/Male/Sliding/KneeSliding/BaseballSlidingSlopeAngle-45SW)
     D="Equal (Object)"(A=@n37, B=/Game/Animation/Game/MainPlayer/Locomotion/Medium/Male/Sliding/KneeSliding/BaseballSlidingSlopeAngle-45W)
     E="Equal (Object)"(A=@n37, B=/Game/Animation/Game/MainPlayer/Locomotion/Medium/Male/Sliding/KneeSliding/BaseballSlidingSlopeAngle45E)
    )
    D:bool=OR Boolean(
     A="Equal (Object)"(A=@n37, B=/Game/Animation/Game/MainPlayer/Locomotion/Medium/Male/Sliding/KneeSliding/BaseballSlidingSlopeAngleC)
     B="Equal (Object)"(A=@n37, B=/Game/Animation/Game/MainPlayer/Locomotion/Medium/Male/Sliding/KneeSliding/BaseballSlidingSlopeAngleE)
     C="Equal (Object)"(A=@n37, B=/Game/Animation/Game/MainPlayer/Locomotion/Medium/Male/Sliding/KneeSliding/BaseballSlidingSlopeAngleS)
     D="Equal (Object)"(A=@n37, B=/Game/Animation/Game/MainPlayer/Locomotion/Medium/Male/Sliding/KneeSliding/BaseballSlidingSlopeAngleSE)
     E="Equal (Object)"(A=@n37, B=/Game/Animation/Game/MainPlayer/Locomotion/Medium/Male/Sliding/KneeSliding/BaseballSlidingSlopeAngleSW)
    )
    E:bool=OR Boolean!(
     A="Equal (Object)"(A=@n37, B=/Game/Animation/Game/MainPlayer/Locomotion/Medium/Male/Sliding/KneeSliding/BaseballSlidingSlopeAngleW)
     B="Equal (Object)"(A=@n37, B=/Game/Animation/Game/MainPlayer/Locomotion/Medium/Male/Sliding/KneeSliding/BaseballToKneeSlidingTransition45E)
     C="Equal (Object)"(A=@n37, B=/Game/Animation/Game/MainPlayer/Locomotion/Medium/Male/Sliding/KneeSliding/BaseballToKneeSlidingTransition45N)
     D="Equal (Object)"(A=@n37, B=/Game/Animation/Game/MainPlayer/Locomotion/Medium/Male/Sliding/KneeSliding/BaseballToKneeSlidingTransition45S)
     E="Equal (Object)"(A=@n37, B=/Game/Animation/Game/MainPlayer/Locomotion/Medium/Male/Sliding/KneeSliding/BaseballToKneeSlidingTransition45W)
     F:bool="Equal (Object)"(A=@n37, B=/Game/Animation/Game/MainPlayer/Locomotion/Medium/Male/Sliding/KneeSliding/BaseballToKneeSlidingTransitionC)
    ) -> (ReturnValue)
   ) -> (ReturnValue)
   AND Boolean!(A=Get bIsCrouching(), B=@n28) -> (ReturnValue)
   Event OnCharacterCustomizationLoaded()
   Get bIsSlopeSliding() -> (bIsSlopeSliding:bool)
  ]}
]
variables [
 BackLeft {type=number}
 BackRight {type=number}
 BakcwardRightMoving {type=boolean}
 baseballSliding {type=boolean}
 crouchBack {type=number}
 Crouched {type=number}
 Crouching_Alpha {type=number}
 crouchLeft {type=number}
 crouchWalking {type=boolean}
 FrontRightCrouchMoving {type=boolean}
 Gravity Override {type=Vector}
 Gravity Override Params {
  type=GravityOverrideParamsStruct
  default {skydive_FloatingDown {gravityOverride=<0 -700 100>, windAmplitude=<200 -500 200>, windOffset=<0 -600 200>}}
 }
 Is_Jumping {type=boolean}
 Jump_Alpha {type=number}
 JumpFall {type=boolean}
 JumpSprint {type=boolean}
 legR_r {type=number}
 legUp_l {type=number}
 legUp_r {type=number}
 movingBack {type=number}
 NewVar_0 {type=SkeletalMeshComponent}
 Not_Crouched {type=number}
 PelvisDive {type=number}
 skelMesh {type=SkeletalMeshComponent}
 SprintSlide {type=boolean}
 Standing Moving {type=boolean}
]
