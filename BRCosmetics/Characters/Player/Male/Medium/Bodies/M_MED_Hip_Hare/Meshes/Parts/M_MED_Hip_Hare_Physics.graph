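# Physics asset for the Hip Hare medium male player cosmetic, defining capsule bodies and joints for its dynamic ears, hair and button bones.
# Two-bone ear chains and two button bones attach to the head with tight angular limits, though the right ear base swings less than the left.
# Three hair bones hang with limited motion from a master hair bone, whose joint to the head has free swing and twist.
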PhysicsAsset /BRCosmetics/Characters/Player/Male/Medium/Bodies/M_MED_Hip_Hare/Meshes/Parts/M_MED_Hip_Hare_Physics
bodies {
 Dyn_Button_Top [{shapeType=Capsule, length=2, radius=1, center=<0.5 0 0>}]
 Dyn_Button_Top1 [{shapeType=Capsule, length=2, radius=1, center=<0.5 0 0>}]
 Dyn_Hair1 [{shapeType=Capsule, length=1, radius=1, center=<0.5 0 0>}]
 Dyn_Hair2 [{shapeType=Capsule, length=1, radius=1, center=<0.5 0 0>}]
 Dyn_Hair3 [{shapeType=Capsule, length=1, radius=1, center=<0.5 0 0>}]
 Dyn_Left_Ear1 [{shapeType=Capsule, length=5, radius=5, center=<1 0 0>, rotation=<rot 90 0 0>}]
 Dyn_Left_Ear2 [{shapeType=Capsule, length=5, radius=5, center=<1 0 0>, rotation=<rot 90 0 0>}]
 Dyn_Master_Hair [{shapeType=Capsule, length=1, radius=1, rotation=<rot 0 -90 0>}]
 Dyn_Right_Ear1 [{shapeType=Capsule, length=5, radius=5, center=<1 0 0>, rotation=<rot 90 0 0>}]
 Dyn_Right_Ear2 [{shapeType=Capsule, length=5, radius=5, center=<1 0 0>, rotation=<rot 90 0 0>}]
 head [{shapeType=Capsule, length=1, radius=1}]
}
constraints [
 {bone1Name=Dyn_Right_Ear1, bone2Name=head, swing1LimitDegrees=10, swing2LimitDegrees=5, twistLimitDegrees=5}
 {bone1Name=Dyn_Right_Ear2, bone2Name=Dyn_Right_Ear1, swing1LimitDegrees=30, swing2LimitDegrees=10, twistLimitDegrees=5}
 {bone1Name=Dyn_Left_Ear1, bone2Name=head, swing1LimitDegrees=30, swing2LimitDegrees=5, twistLimitDegrees=5}
 {bone1Name=Dyn_Left_Ear2, bone2Name=Dyn_Left_Ear1, swing1LimitDegrees=30, swing2LimitDegrees=10, twistLimitDegrees=5}
 {bone1Name=Dyn_Button_Top1, bone2Name=head, swing1LimitDegrees=1, swing2LimitDegrees=10, twistLimitDegrees=10}
 {bone1Name=Dyn_Button_Top, bone2Name=head, swing1LimitDegrees=1, swing2LimitDegrees=10, twistLimitDegrees=10}
 {bone1Name=Dyn_Hair1, bone2Name=Dyn_Master_Hair, swing1LimitDegrees=10, swing2LimitDegrees=10, twistLimitDegrees=10}
 {bone1Name=Dyn_Hair2, bone2Name=Dyn_Master_Hair, swing1LimitDegrees=10, swing2LimitDegrees=10, twistLimitDegrees=10}
 {bone1Name=Dyn_Hair3, bone2Name=Dyn_Master_Hair, swing1LimitDegrees=10, swing2LimitDegrees=10, twistLimitDegrees=10}
 {
  bone1Name=Dyn_Master_Hair
  bone2Name=head
  swing1LimitDegrees=45
  swing1Motion=Free
  swing2LimitDegrees=45
  swing2Motion=Free
  twistLimitDegrees=45
  twistMotion=Free
 }
]
