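# Head-part AnimBP for the Hip Hare male cosmetic body, following the attached parent mesh and adding physics motion.
# The output pose runs Copy Pose From Mesh into a RigidBody node simulating the Hip Hare physics asset in head-bone space with overridden gravity.
# A chain of additive ear and head bone tweaks, toggled by vehicle and vehicle-seat flags, is not connected to the output.
# The event graph has no logic.
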AnimBlueprint /BRCosmetics/Characters/Player/Male/Medium/Bodies/M_MED_Hip_Hare/Meshes/Parts/M_MED_Hip_Hare_Head_AnimBP
target_skeleton=&/M_MED_Hip_Hare_Skeleton
graphs [
 AnimGraph {nodes [
   n1 = Get bIsInVehicle() -> (bIsInVehicle:bool)
   n2 = "Transform (Modify) Bone - Bone: Dyn_Right_Ear1"!(ComponentPose, bAlphaBoolEnabled:bool=@n1, Translation="0,0,0", Rotation="50,90,30", Scale="1,1,1") -> (Pose) {
    alphaInputType=Bool
    rotationMode=BMM_Additive
    rotationSpace=BCS_BoneSpace
    scaleMode=BMM_Additive
    scaleSpace=BCS_BoneSpace
    alphaBoolBlend {blendOption=Linear}
    boneToModify {boneName=Dyn_Right_Ear1}
    rotation=<rot 50 90 30>
   }
   n3 = "Transform (Modify) Bone - Bone: Dyn_Left_Ear1"!(ComponentPose=@n2, bAlphaBoolEnabled:bool=@n1, Translation="0,0,0", Rotation="-60,120,-40", Scale="1,1,1") -> (Pose) {
    alphaInputType=Bool
    rotationMode=BMM_Additive
    rotationSpace=BCS_BoneSpace
    scaleMode=BMM_Additive
    scaleSpace=BCS_BoneSpace
    alphaBoolBlend {blendOption=Linear}
    boneToModify {boneName=Dyn_Left_Ear1}
    rotation=<rot -60 120 -40>
   }
   n4 = "Transform (Modify) Bone - Bone: head"!(ComponentPose=@n3, bAlphaBoolEnabled:bool=@n1, Translation="0,0,0", Rotation="0,0,0", Scale="0.9,0.9,0.9") -> (Pose) {
    alphaInputType=Bool
    rotationMode=BMM_Additive
    rotationSpace=BCS_BoneSpace
    scaleMode=BMM_Additive
    scaleSpace=BCS_BoneSpace
    alphaBoolBlend {blendOption=Linear}
    boneToModify {boneName=head}
    scale=<0.9 0.9 0.9>
   }
   n5 = Get bIsInVehicleSeat() -> (bIsInVehicleSeat:bool)
   n6 = "Transform (Modify) Bone - Bone: Dyn_Right_Ear1"(ComponentPose=@n4, bAlphaBoolEnabled=@n5, Translation="0,0,0", Rotation="50,90,30", Scale="1,1,1") {
    alphaInputType=Bool
    rotationMode=BMM_Additive
    rotationSpace=BCS_BoneSpace
    scaleMode=BMM_Additive
    scaleSpace=BCS_BoneSpace
    alphaBoolBlend {blendOption=Linear}
    boneToModify {boneName=Dyn_Right_Ear1}
    rotation=<rot 50 90 30>
   }
   n7 = "Transform (Modify) Bone - Bone: Dyn_Left_Ear1"(ComponentPose=@n6, bAlphaBoolEnabled=@n5, Translation="0,0,0", Rotation="-60,120,-40", Scale="1,1,1") {
    alphaInputType=Bool
    rotationMode=BMM_Additive
    rotationSpace=BCS_BoneSpace
    scaleMode=BMM_Additive
    scaleSpace=BCS_BoneSpace
    alphaBoolBlend {blendOption=Linear}
    boneToModify {boneName=Dyn_Left_Ear1}
    rotation=<rot -60 120 -40>
   }
   "Transform (Modify) Bone - Bone: head"(ComponentPose=@n7, bAlphaBoolEnabled=@n5, Translation="0,0,0", Rotation="0,0,0", Scale="0.9,0.9,0.9") {
    alphaInputType=Bool
    rotationMode=BMM_Additive
    rotationSpace=BCS_BoneSpace
    scaleMode=BMM_Additive
    scaleSpace=BCS_BoneSpace
    alphaBoolBlend {blendOption=Linear}
    boneToModify {boneName=head}
    scale=<0.9 0.9 0.9>
   }
   n8 = RigidBody(ComponentPose=Local To Component(LocalPose=Copy Pose From Mesh() {bUseAttachedParent=true})) {
    bFreezeIncomingPoseOnStart=true
    bOverrideWorldGravity=true
    overridePhysicsAsset=&/M_MED_Hip_Hare_Physics
    simulationSpace=BaseBoneSpace
    alphaBoolBlend {blendOption=Linear}
    baseBoneRef {boneName=head}
    componentAppliedLinearAccClamp=<12000 12000 12000>
    componentLinearAccScale=<0.1 1 1>
    componentLinearVelScale=<0.1 0.1 0.1>
    overrideWorldGravity {z=-980}
   }
   Output Pose(Result=Component To Local(ComponentPose=@n8)) {layerGroup=None}
  ]}
 EventGraph {nodes [Event BlueprintUpdateAnimation()]}
]
variables []
